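# Physics asset for the M_MED_Zest male medium player body, set up for dynamic jacket motion against the lower body.
# Four jacket chains (back and mid, left and right) each link four capsules to a tiny MSTR anchor constrained to spine_01, with swing widening toward the tip and twist nearly locked.
# Pelvis, lower spine, thigh and calf capsules with free-rotation constraints, plus two root spheres, one very large and offset downward, form the body shapes.
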PhysicsAsset /BRCosmetics/Characters/Player/Male/Medium/Bodies/M_MED_Zest/Meshes/M_MED_Zest_Physics
bodies {
 calf_l [{shapeType=Capsule, length=53.1892, radius=8.5, shapeName=calf_r_capsule, center=<-27.3745 -2.0001 0.1912>, rotation=<rot 90 0 0>}]
 calf_r [{shapeType=Capsule, length=61.4375, radius=9.4136, shapeName=calf_r_capsule, center=<33.7477 1.9999 -0.1912>, rotation=<rot -90 0 0>}]
 dyn_jacket_bk_l_01 [{shapeType=Capsule, length=21, radius=3, shapeName=dyn_jacket_fr_l_1_capsule, center=<15 0 0>}]
 dyn_jacket_bk_l_02 [{shapeType=Capsule, length=27.125, radius=3.5, shapeName=dyn_jacket_fr_l_1_capsule, center=<15 0 0>}]
 dyn_jacket_bk_l_03 [{shapeType=Capsule, length=33.625, radius=4.5, shapeName=dyn_jacket_fr_l_1_capsule, center=<15 0 0>}]
 dyn_jacket_bk_l_04 [{shapeType=Capsule, length=33.125, radius=5, shapeName=dyn_jacket_fr_l_1_capsule, center=<17 0 0>}]
 dyn_jacket_bk_l_MSTR [{shapeType=Capsule, length=0.505, radius=0.505, shapeName=dyn_jacket_bk_l_MSTR_capsule}]
 dyn_jacket_bk_r_01 [{shapeType=Capsule, length=21, radius=3, shapeName=dyn_jacket_fr_l_1_capsule, center=<15 0 0>}]
 dyn_jacket_bk_r_02 [{shapeType=Capsule, length=26.0625, radius=3.8125, shapeName=dyn_jacket_fr_l_1_capsule, center=<15 0 0>}]
 dyn_jacket_bk_r_03 [{shapeType=Capsule, length=33.625, radius=4.5, shapeName=dyn_jacket_fr_l_1_capsule, center=<15 0 0>}]
 dyn_jacket_bk_r_04 [{shapeType=Capsule, length=33.125, radius=5, shapeName=dyn_jacket_fr_l_1_capsule, center=<17 0 0>}]
 dyn_jacket_bk_r_MSTR [{shapeType=Capsule, length=0.505, radius=0.505, shapeName=dyn_jacket_bk_r_MSTR_capsule}]
 dyn_jacket_mid_l_01 [{shapeType=Capsule, length=18.25, radius=3, shapeName=dyn_jacket_fr_l_1_capsule, center=<15 0 0>}]
 dyn_jacket_mid_l_02 [{shapeType=Capsule, length=23.875, radius=3.8125, shapeName=dyn_jacket_fr_l_1_capsule, center=<15 0 0>}]
 dyn_jacket_mid_l_03 [{shapeType=Capsule, length=27.0625, radius=4.6875, shapeName=dyn_jacket_fr_l_1_capsule, center=<15 0 0>}]
 dyn_jacket_mid_l_04 [{shapeType=Capsule, length=26.4375, radius=5.5, shapeName=dyn_jacket_fr_l_1_capsule, center=<22.7408 0 -0.7344>}]
 dyn_jacket_mid_l_MSTR [{shapeType=Capsule, length=0.505, radius=0.505, shapeName=dyn_jacket_mid_l_MSTR_capsule}]
 dyn_jacket_mid_r_01 [{shapeType=Capsule, length=18.25, radius=3, shapeName=dyn_jacket_fr_l_1_capsule, center=<15 0 0>}]
 dyn_jacket_mid_r_02 [{shapeType=Capsule, length=23.875, radius=3.8125, shapeName=dyn_jacket_fr_l_1_capsule, center=<15 0 0>}]
 dyn_jacket_mid_r_03 [{shapeType=Capsule, length=27.0625, radius=4.6875, shapeName=dyn_jacket_fr_l_1_capsule, center=<15 0 0>}]
 dyn_jacket_mid_r_04 [{shapeType=Capsule, length=26.4375, radius=5.5, shapeName=dyn_jacket_fr_l_1_capsule, center=<22.7408 0 -0.7344>}]
 dyn_jacket_mid_r_MSTR [{shapeType=Capsule, length=0.505, radius=0.505, shapeName=dyn_jacket_mid_r_MSTR_capsule}]
 pelvis [{shapeType=Capsule, length=10.9902, radius=12.299, shapeName=pelvis_capsule, center=<-0.8716 0.169 -0.0001>, rotation=<rot 90 0 -10.0001>}]
 root [
  {shapeType=Sphere, radius=32, shapeName=root_sphere, center=<0.0002 2.5589 8.5972>}
  {shapeType=Sphere, radius=183.4375, shapeName=root_sphere, center=<0 -8.6403 -172.3024>}
 ]
 spine_01 [{shapeType=Capsule, length=16.9245, radius=11.6914, shapeName=spine_01_capsule, center=<2.7881 0 0>, rotation=<rot 90 0 0>}]
 thigh_l [{shapeType=Capsule, length=31.1446, radius=8.5, shapeName=thigh_r_capsule, center=<-20.8013 -0.5 0.6036>, rotation=<rot 86.1544 0 -0.0002>}]
 thigh_r [{shapeType=Capsule, length=31.6446, radius=8.5, shapeName=thigh_r_capsule, center=<20.806 0.5 -0.8464>, rotation=<rot -88.898 0.0001 -0.0001>}]
}
constraints [
 {
  bone1Name=pelvis
  bone2Name=root
  swing1LimitDegrees=45
  swing1Motion=Free
  swing2LimitDegrees=45
  swing2Motion=Free
  twistLimitDegrees=45
  twistMotion=Free
 }
 {
  bone1Name=spine_01
  bone2Name=pelvis
  swing1LimitDegrees=45
  swing1Motion=Free
  swing2LimitDegrees=45
  swing2Motion=Free
  twistLimitDegrees=45
  twistMotion=Free
 }
 {
  bone1Name=dyn_jacket_mid_r_MSTR
  bone2Name=spine_01
  swing1LimitDegrees=45
  swing1Motion=Free
  swing2LimitDegrees=45
  swing2Motion=Free
  twistLimitDegrees=45
  twistMotion=Free
 }
 {bone1Name=dyn_jacket_mid_r_01, bone2Name=dyn_jacket_mid_r_MSTR, swing1LimitDegrees=20, swing2LimitDegrees=15, twistLimitDegrees=1}
 {bone1Name=dyn_jacket_mid_r_02, bone2Name=dyn_jacket_mid_r_01, swing1LimitDegrees=30, swing2LimitDegrees=15, twistLimitDegrees=1}
 {bone1Name=dyn_jacket_mid_r_03, bone2Name=dyn_jacket_mid_r_02, swing1LimitDegrees=30, swing2LimitDegrees=15, twistLimitDegrees=1}
 {bone1Name=dyn_jacket_mid_r_04, bone2Name=dyn_jacket_mid_r_03, swing1LimitDegrees=35, swing2LimitDegrees=15, twistLimitDegrees=1}
 {
  bone1Name=dyn_jacket_bk_r_MSTR
  bone2Name=spine_01
  swing1LimitDegrees=45
  swing1Motion=Free
  swing2LimitDegrees=45
  swing2Motion=Free
  twistLimitDegrees=45
  twistMotion=Free
 }
 {bone1Name=dyn_jacket_bk_r_01, bone2Name=dyn_jacket_bk_r_MSTR, swing1LimitDegrees=20, swing2LimitDegrees=15, twistLimitDegrees=1}
 {bone1Name=dyn_jacket_bk_r_02, bone2Name=dyn_jacket_bk_r_01, swing1LimitDegrees=30, swing2LimitDegrees=15, twistLimitDegrees=1}
 {bone1Name=dyn_jacket_bk_r_03, bone2Name=dyn_jacket_bk_r_02, swing1LimitDegrees=30, swing2LimitDegrees=15, twistLimitDegrees=1}
 {bone1Name=dyn_jacket_bk_r_04, bone2Name=dyn_jacket_bk_r_03, swing1LimitDegrees=35, swing2LimitDegrees=15, twistLimitDegrees=1}
 {
  bone1Name=dyn_jacket_bk_l_MSTR
  bone2Name=spine_01
  swing1LimitDegrees=45
  swing1Motion=Free
  swing2LimitDegrees=45
  swing2Motion=Free
  twistLimitDegrees=45
  twistMotion=Free
 }
 {bone1Name=dyn_jacket_bk_l_01, bone2Name=dyn_jacket_bk_l_MSTR, swing1LimitDegrees=20, swing2LimitDegrees=15, twistLimitDegrees=1}
 {bone1Name=dyn_jacket_bk_l_02, bone2Name=dyn_jacket_bk_l_01, swing1LimitDegrees=30, swing2LimitDegrees=15, twistLimitDegrees=1}
 {bone1Name=dyn_jacket_bk_l_03, bone2Name=dyn_jacket_bk_l_02, swing1LimitDegrees=30, swing2LimitDegrees=15, twistLimitDegrees=1}
 {bone1Name=dyn_jacket_bk_l_04, bone2Name=dyn_jacket_bk_l_03, swing1LimitDegrees=35, swing2LimitDegrees=15, twistLimitDegrees=1}
 {
  bone1Name=dyn_jacket_mid_l_MSTR
  bone2Name=spine_01
  swing1LimitDegrees=45
  swing1Motion=Free
  swing2LimitDegrees=45
  swing2Motion=Free
  twistLimitDegrees=45
  twistMotion=Free
 }
 {bone1Name=dyn_jacket_mid_l_01, bone2Name=dyn_jacket_mid_l_MSTR, swing1LimitDegrees=20, swing2LimitDegrees=15, twistLimitDegrees=1}
 {bone1Name=dyn_jacket_mid_l_02, bone2Name=dyn_jacket_mid_l_01, swing1LimitDegrees=30, swing2LimitDegrees=15, twistLimitDegrees=1}
 {bone1Name=dyn_jacket_mid_l_03, bone2Name=dyn_jacket_mid_l_02, swing1LimitDegrees=30, swing2LimitDegrees=15, twistLimitDegrees=1}
 {bone1Name=dyn_jacket_mid_l_04, bone2Name=dyn_jacket_mid_l_03, swing1LimitDegrees=35, swing2LimitDegrees=15, twistLimitDegrees=1}
 {
  bone1Name=thigh_l
  bone2Name=pelvis
  swing1LimitDegrees=45
  swing1Motion=Free
  swing2LimitDegrees=45
  swing2Motion=Free
  twistLimitDegrees=45
  twistMotion=Free
 }
 {
  bone1Name=calf_l
  bone2Name=thigh_l
  swing1LimitDegrees=45
  swing1Motion=Free
  swing2LimitDegrees=45
  swing2Motion=Free
  twistLimitDegrees=45
  twistMotion=Free
 }
 {
  bone1Name=thigh_r
  bone2Name=pelvis
  swing1LimitDegrees=45
  swing1Motion=Free
  swing2LimitDegrees=45
  swing2Motion=Free
  twistLimitDegrees=45
  twistMotion=Free
 }
 {
  bone1Name=calf_r
  bone2Name=thigh_r
  swing1LimitDegrees=45
  swing1Motion=Free
  swing2LimitDegrees=45
  swing2Motion=Free
  twistLimitDegrees=45
  twistMotion=Free
 }
]
