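# Physics asset for the M_MED_Zest male body cosmetic, built to simulate a dynamic jacket against the lower body.
# Pelvis, spine_01, legs and feet carry capsules with free-rotation joints, and root carries two spheres, one very large and offset.
# Back and side four-bone chains plus single-box front flaps hang from spine_01 via tiny MSTR bodies, with limited swing and tight twist.
# Free-rotation cross-constraints tie adjacent back and side chains at their third segments.
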PhysicsAsset /BRCosmetics/Characters/Player/Male/Medium/Bodies/M_MED_Zest/Meshes/M_MED_Zest_Physics1
bodies {
 calf_l [{shapeType=Capsule, length=30.9321, radius=8.8593, center=<-25 -2 0>, rotation=<rot 90 0 0>}]
 calf_r [{shapeType=Capsule, length=30.9321, radius=8.8593, center=<25 2 0.1223>, rotation=<rot 90 0 0>}]
 dyn_jacket_bk_l_01 [{shapeType=Capsule, length=9.225, radius=6, shapeName=dyn_jacketBack_01_r_capsule, center=<8 4 0>}]
 dyn_jacket_bk_l_02 [{shapeType=Capsule, length=10.1, radius=8, shapeName=dyn_jacketBack_02_r_capsule, center=<8 6 0>}]
 dyn_jacket_bk_l_03 [{shapeType=Capsule, length=12, radius=9, shapeName=dyn_jacketBack_03_r_capsule, center=<8 6 0>}]
 dyn_jacket_bk_l_04 [{shapeType=Capsule, length=16.158, radius=9.5, shapeName=dyn_jacketBack_04_r_capsule, center=<8 7 0>}]
 dyn_jacket_bk_l_MSTR [{shapeType=Capsule, length=0.505, radius=0.505, shapeName=dyn_jacket_bk_l_MSTR_capsule}]
 dyn_jacket_bk_r_01 [{shapeType=Capsule, length=9.225, radius=6, shapeName=dyn_jacketBack_01_r_capsule, center=<8 4 0>}]
 dyn_jacket_bk_r_02 [{shapeType=Capsule, length=10.1, radius=8, shapeName=dyn_jacketBack_02_r_capsule, center=<8 6 0>}]
 dyn_jacket_bk_r_03 [{shapeType=Capsule, length=12, radius=9, shapeName=dyn_jacketBack_03_r_capsule, center=<8 6 0>}]
 dyn_jacket_bk_r_04 [{shapeType=Capsule, length=16.158, radius=9.5, shapeName=dyn_jacketBack_04_r_capsule, center=<8 7 0>}]
 dyn_jacket_bk_r_MSTR [{shapeType=Capsule, length=0.505, radius=0.505, shapeName=dyn_jacket_bk_r_MSTR_capsule}]
 dyn_jacket_fr_l_01 [{shapeType=Box, extentX=2.5, extentY=15, extentZ=10, shapeName=dyn_collar_r_01_box, center=<6.7876 0 0>, rotation=<rot 0 90 0>}]
 dyn_jacket_fr_l_MSTR [{shapeType=Capsule, length=0.505, radius=0.505, shapeName=dyn_jacket_fr_l_MSTR_capsule, rotation=<rot 0 -90 0>}]
 dyn_jacket_fr_r_01 [{shapeType=Box, extentX=2.5, extentY=15, extentZ=10, shapeName=dyn_collar_r_01_box, center=<6.7876 0 0>, rotation=<rot 0 90 0>}]
 dyn_jacket_fr_r_MSTR [{shapeType=Capsule, length=0.505, radius=0.505, shapeName=dyn_jacket_fr_r_MSTR_capsule, rotation=<rot 0 -90 0>}]
 dyn_jacket_mid_l_01 [
  {shapeType=Capsule, length=16.525, radius=5.25, shapeName=dyn_jacketSide_01_r_capsule, center=<8 3.5 0>}
  {shapeType=Capsule, length=16.525, radius=5.25, shapeName=dyn_jacket_mid_l_01_capsule, center=<18 1.8414 -4.9869>, rotation=<rot 85 0 0>}
 ]
 dyn_jacket_mid_l_02 [{shapeType=Capsule, length=16.05, radius=8.25, shapeName=dyn_jacketSide_02_r_capsule, center=<8 4 -5>}]
 dyn_jacket_mid_l_03 [{shapeType=Capsule, length=19.35, radius=9, shapeName=dyn_jacketSide_03_r_capsule, center=<10 5 -4.5>}]
 dyn_jacket_mid_l_04 [{shapeType=Capsule, length=16.85, radius=9.25, shapeName=dyn_jacketSide_03_r_capsule, center=<14 5 -4.5>}]
 dyn_jacket_mid_l_MSTR [{shapeType=Capsule, length=0.505, radius=0.505, shapeName=dyn_jacket_mid_l_MSTR_capsule}]
 dyn_jacket_mid_r_01 [{shapeType=Capsule, length=12.775, radius=5.5, shapeName=dyn_jacketSide_01_r_capsule, center=<8 4 0>}]
 dyn_jacket_mid_r_02 [{shapeType=Capsule, length=12.55, radius=8.5, shapeName=dyn_jacketSide_02_r_capsule, center=<8 4 5>}]
 dyn_jacket_mid_r_03 [{shapeType=Capsule, length=16.85, radius=9, shapeName=dyn_jacketSide_03_r_capsule, center=<10 5 4.5>}]
 dyn_jacket_mid_r_04 [{shapeType=Capsule, length=16.1, radius=9.5, shapeName=dyn_jacketSide_03_r_capsule, center=<14 5 4.5>}]
 dyn_jacket_mid_r_MSTR [{shapeType=Capsule, length=0.505, radius=0.505, shapeName=dyn_jacket_mid_r_MSTR_capsule}]
 foot_l [{shapeType=Capsule, length=17.5237, radius=11.7507, center=<-3.4271 8.491 0.1091>, rotation=<rot 0.7349 0.9586 -85.8278>}]
 foot_r [{shapeType=Capsule, length=21.5376, radius=11.7675, center=<3.4917 -9.9816 -0.1619>, rotation=<rot 0.7221 1.1909 -85.9009>}]
 pelvis [
  {shapeType=Capsule, length=10.0499, radius=11.7455, center=<-5.6598 -1.6342 0>}
  {shapeType=Capsule, length=15.5495, radius=13.125, shapeName=spine_01_capsule, center=<0.3835 0 0>, rotation=<rot 90 0 0>}
 ]
 root [
  {shapeType=Sphere, radius=187.1686, shapeName=root_sphere, center=<-3.4923 -2.7141 -174.8455>}
  {shapeType=Sphere, radius=24.9585, shapeName=root_sphere, center=<0 2.4787 8.5972>}
 ]
 spine_01 [{shapeType=Capsule, length=16.6745, radius=12, shapeName=spine_01_capsule, center=<2.7881 0 0>, rotation=<rot 90 0 0>}]
 thigh_l [{shapeType=Capsule, length=33.2696, radius=9.75, shapeName=thigh_r_capsule, center=<-20.8013 -1 0.6036>, rotation=<rot 86.1544 0 -0.0002>}]
 thigh_r [{shapeType=Capsule, length=33.7696, radius=9.5, shapeName=thigh_r_capsule, center=<20.806 1 -0.8464>, rotation=<rot -88.898 0.0001 -0.0001>}]
}
constraints [
 {
  bone1Name=pelvis
  bone2Name=root
  swing1LimitDegrees=45
  swing1Motion=Free
  swing2LimitDegrees=45
  swing2Motion=Free
  twistLimitDegrees=45
  twistMotion=Free
 }
 {
  bone1Name=spine_01
  bone2Name=pelvis
  swing1LimitDegrees=45
  swing1Motion=Free
  swing2LimitDegrees=45
  swing2Motion=Free
  twistLimitDegrees=45
  twistMotion=Free
 }
 {
  bone1Name=dyn_jacket_mid_r_MSTR
  bone2Name=spine_01
  swing1LimitDegrees=45
  swing1Motion=Free
  swing2LimitDegrees=45
  swing2Motion=Free
  twistLimitDegrees=45
  twistMotion=Free
 }
 {bone1Name=dyn_jacket_mid_r_01, bone2Name=dyn_jacket_mid_r_MSTR, swing1LimitDegrees=20, swing2LimitDegrees=45, twistLimitDegrees=5}
 {bone1Name=dyn_jacket_mid_r_02, bone2Name=dyn_jacket_mid_r_01, swing1LimitDegrees=45, swing2LimitDegrees=45, twistLimitDegrees=2}
 {bone1Name=dyn_jacket_mid_r_03, bone2Name=dyn_jacket_mid_r_02, swing1LimitDegrees=55, swing2LimitDegrees=55, twistLimitDegrees=5}
 {bone1Name=dyn_jacket_mid_r_04, bone2Name=dyn_jacket_mid_r_03, swing1LimitDegrees=55, swing2LimitDegrees=55, twistLimitDegrees=5}
 {
  bone1Name=dyn_jacket_bk_r_MSTR
  bone2Name=spine_01
  swing1LimitDegrees=45
  swing1Motion=Free
  swing2LimitDegrees=45
  swing2Motion=Free
  twistLimitDegrees=45
  twistMotion=Free
 }
 {bone1Name=dyn_jacket_bk_r_01, bone2Name=dyn_jacket_bk_r_MSTR, swing1LimitDegrees=15, swing2LimitDegrees=45, twistLimitDegrees=5}
 {bone1Name=dyn_jacket_bk_r_02, bone2Name=dyn_jacket_bk_r_01, swing1LimitDegrees=35, swing2LimitDegrees=35, twistLimitDegrees=2}
 {bone1Name=dyn_jacket_bk_r_03, bone2Name=dyn_jacket_bk_r_02, swing1LimitDegrees=45, swing2LimitDegrees=45, twistLimitDegrees=5}
 {bone1Name=dyn_jacket_bk_r_04, bone2Name=dyn_jacket_bk_r_03, swing1LimitDegrees=45, swing2LimitDegrees=45, twistLimitDegrees=5}
 {
  bone1Name=dyn_jacket_bk_l_MSTR
  bone2Name=spine_01
  swing1LimitDegrees=45
  swing1Motion=Free
  swing2LimitDegrees=45
  swing2Motion=Free
  twistLimitDegrees=45
  twistMotion=Free
 }
 {bone1Name=dyn_jacket_bk_l_01, bone2Name=dyn_jacket_bk_l_MSTR, swing1LimitDegrees=15, swing2LimitDegrees=45, twistLimitDegrees=5}
 {bone1Name=dyn_jacket_bk_l_02, bone2Name=dyn_jacket_bk_l_01, swing1LimitDegrees=35, swing2LimitDegrees=35, twistLimitDegrees=2}
 {bone1Name=dyn_jacket_bk_l_03, bone2Name=dyn_jacket_bk_l_02, swing1LimitDegrees=45, swing2LimitDegrees=45, twistLimitDegrees=5}
 {bone1Name=dyn_jacket_bk_l_04, bone2Name=dyn_jacket_bk_l_03, swing1LimitDegrees=45, swing2LimitDegrees=45, twistLimitDegrees=5}
 {
  bone1Name=dyn_jacket_mid_l_MSTR
  bone2Name=spine_01
  swing1LimitDegrees=45
  swing1Motion=Free
  swing2LimitDegrees=45
  swing2Motion=Free
  twistLimitDegrees=45
  twistMotion=Free
 }
 {bone1Name=dyn_jacket_mid_l_01, bone2Name=dyn_jacket_mid_l_MSTR, swing1LimitDegrees=20, swing2LimitDegrees=45, twistLimitDegrees=5}
 {bone1Name=dyn_jacket_mid_l_02, bone2Name=dyn_jacket_mid_l_01, swing1LimitDegrees=45, swing2LimitDegrees=45, twistLimitDegrees=2}
 {bone1Name=dyn_jacket_mid_l_03, bone2Name=dyn_jacket_mid_l_02, swing1LimitDegrees=55, swing2LimitDegrees=55, twistLimitDegrees=5}
 {bone1Name=dyn_jacket_mid_l_04, bone2Name=dyn_jacket_mid_l_03, swing1LimitDegrees=55, swing2LimitDegrees=55, twistLimitDegrees=5}
 {
  bone1Name=thigh_l
  bone2Name=pelvis
  swing1LimitDegrees=45
  swing1Motion=Free
  swing2LimitDegrees=45
  swing2Motion=Free
  twistLimitDegrees=45
  twistMotion=Free
 }
 {
  bone1Name=calf_l
  bone2Name=thigh_l
  swing1LimitDegrees=45
  swing1Motion=Free
  swing2LimitDegrees=45
  swing2Motion=Free
  twistLimitDegrees=45
  twistMotion=Free
 }
 {
  bone1Name=thigh_r
  bone2Name=pelvis
  swing1LimitDegrees=45
  swing1Motion=Free
  swing2LimitDegrees=45
  swing2Motion=Free
  twistLimitDegrees=45
  twistMotion=Free
 }
 {
  bone1Name=calf_r
  bone2Name=thigh_r
  swing1LimitDegrees=45
  swing1Motion=Free
  swing2LimitDegrees=45
  swing2Motion=Free
  twistLimitDegrees=45
  twistMotion=Free
 }
 {
  bone1Name=foot_r
  bone2Name=calf_r
  swing1LimitDegrees=45
  swing1Motion=Free
  swing2LimitDegrees=45
  swing2Motion=Free
  twistLimitDegrees=45
  twistMotion=Free
 }
 {
  bone1Name=foot_l
  bone2Name=calf_l
  swing1LimitDegrees=45
  swing1Motion=Free
  swing2LimitDegrees=45
  swing2Motion=Free
  twistLimitDegrees=45
  twistMotion=Free
 }
 {
  bone1Name=dyn_jacket_bk_l_03
  bone2Name=dyn_jacket_bk_r_03
  swing1LimitDegrees=45
  swing1Motion=Free
  swing2LimitDegrees=45
  swing2Motion=Free
  twistLimitDegrees=45
  twistMotion=Free
 }
 {
  bone1Name=dyn_jacket_mid_r_03
  bone2Name=dyn_jacket_bk_r_03
  swing1LimitDegrees=45
  swing1Motion=Free
  swing2LimitDegrees=45
  swing2Motion=Free
  twistLimitDegrees=45
  twistMotion=Free
 }
 {
  bone1Name=dyn_jacket_bk_l_03
  bone2Name=dyn_jacket_mid_l_03
  swing1LimitDegrees=45
  swing1Motion=Free
  swing2LimitDegrees=45
  swing2Motion=Free
  twistLimitDegrees=45
  twistMotion=Free
 }
 {
  bone1Name=dyn_jacket_fr_r_MSTR
  bone2Name=spine_01
  swing1LimitDegrees=45
  swing1Motion=Free
  swing2LimitDegrees=45
  swing2Motion=Free
  twistLimitDegrees=45
  twistMotion=Free
 }
 {
  bone1Name=dyn_jacket_fr_l_MSTR
  bone2Name=spine_01
  swing1LimitDegrees=45
  swing1Motion=Free
  swing2LimitDegrees=45
  swing2Motion=Free
  twistLimitDegrees=45
  twistMotion=Free
 }
 {bone1Name=dyn_jacket_fr_l_01, bone2Name=dyn_jacket_fr_l_MSTR, swing1LimitDegrees=35, swing2LimitDegrees=35, twistLimitDegrees=5}
 {bone1Name=dyn_jacket_fr_r_01, bone2Name=dyn_jacket_fr_r_MSTR, swing1LimitDegrees=35, swing2LimitDegrees=35, twistLimitDegrees=5}
]
