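# Physics asset for the collar of the M_MED_Zest male medium player cosmetic, covering left and right dyn_collar bone chains.
# Each collar capsule has tight swing and twist limits against a small MSTR box body, which links to spine_05 with free rotation.
# Pelvis, spine_05 and upper-arm capsules add body shapes around the collar, joined by free-rotation constraints.
# The right upper arm carries two offset capsules, unlike the left's single one.
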PhysicsAsset /BRCosmetics/Characters/Player/Male/Medium/Bodies/M_MED_Zest/Meshes/M_MED_Zest_Physics_Collar
bodies {
 dyn_collar_l_01 [{shapeType=Capsule, length=15.125, radius=5, shapeName=dyn_collar_r_01_capsule, center=<7 -2.25 4.5>, rotation=<rot 45 0 0>}]
 dyn_collar_l_MSTR [{shapeType=Box, extentX=1.01, extentY=1.01, extentZ=1.01, shapeName=dyn_collar_l_MSTR_box, rotation=<rot 0 -90 0>}]
 dyn_collar_r_01 [{shapeType=Capsule, length=15.125, radius=5, shapeName=dyn_collar_r_01_capsule, center=<7 -2.25 -4.5>, rotation=<rot -45 0 0>}]
 dyn_collar_r_MSTR [{shapeType=Box, extentX=1.01, extentY=1.01, extentZ=1.01, shapeName=dyn_collar_r_MSTR_box, rotation=<rot 0 -90 0>}]
 pelvis [{shapeType=Capsule, length=15.5495, radius=13.125, shapeName=spine_01_capsule, center=<0.3835 0 0>, rotation=<rot 90 0 0>}]
 spine_05 [{shapeType=Capsule, length=1, radius=1, shapeName=spine_05_capsule, center=<6.1449 -0.2599 0.9508>, rotation=<rot 74.3126 105.7595 -164.7998>}]
 upperarm_l [{shapeType=Capsule, length=30, radius=7.5, shapeName=upperarm_r_capsule, center=<15 0 -1>, rotation=<rot 90 0 0>}]
 upperarm_r [
  {shapeType=Capsule, length=30, radius=8, shapeName=upperarm_r_capsule, center=<-12.7012 -0.272 1.2602>, rotation=<rot 90 0 0>}
  {shapeType=Capsule, length=30, radius=8, shapeName=upperarm_r_capsule, center=<-12.7012 10.3688 1.2603>, rotation=<rot 90 0 0>}
 ]
}
constraints [
 {
  bone1Name=spine_05
  bone2Name=pelvis
  swing1LimitDegrees=45
  swing1Motion=Free
  swing2LimitDegrees=45
  swing2Motion=Free
  twistLimitDegrees=45
  twistMotion=Free
 }
 {
  bone1Name=upperarm_l
  bone2Name=spine_05
  swing1LimitDegrees=45
  swing1Motion=Free
  swing2LimitDegrees=45
  swing2Motion=Free
  twistLimitDegrees=45
  twistMotion=Free
 }
 {
  bone1Name=upperarm_r
  bone2Name=spine_05
  swing1LimitDegrees=45
  swing1Motion=Free
  swing2LimitDegrees=45
  swing2Motion=Free
  twistLimitDegrees=45
  twistMotion=Free
 }
 {
  bone1Name=dyn_collar_l_MSTR
  bone2Name=spine_05
  swing1LimitDegrees=45
  swing1Motion=Free
  swing2LimitDegrees=45
  swing2Motion=Free
  twistLimitDegrees=45
  twistMotion=Free
 }
 {bone1Name=dyn_collar_l_01, bone2Name=dyn_collar_l_MSTR, swing1LimitDegrees=35, swing2LimitDegrees=10, twistLimitDegrees=2}
 {
  bone1Name=dyn_collar_r_MSTR
  bone2Name=spine_05
  swing1LimitDegrees=45
  swing1Motion=Free
  swing2LimitDegrees=45
  swing2Motion=Free
  twistLimitDegrees=45
  twistMotion=Free
 }
 {bone1Name=dyn_collar_r_01, bone2Name=dyn_collar_r_MSTR, swing1LimitDegrees=35, swing2LimitDegrees=10, twistLimitDegrees=2}
]
