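# Post-process style AnimBlueprint for the M_MED_Zest body cosmetic that adds jacket and collar motion over the parent mesh's copied pose.
# Two offset-plus-RigidBody jacket branches switch on skydiving, then collar correctives, collar simulation and front-jacket rotation constraints follow.
# The EventGraph derives corrective alphas from thigh, head and arm rotations, gated by movement or weapon aiming where relevant, and updates state-based gravity for the jacket.
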AnimBlueprint /BRCosmetics/Characters/Player/Male/Medium/Bodies/M_MED_Zest/Meshes/M_MED_Zest_AnimBP
target_skeleton=&/M_MED_Zest_Skeleton
graphs [
 AnimGraph {nodes [
   n1 = Get bIsCrouching()
   n2 = "Transform (Modify) Bone - Bone: dyn_jacket_mid_r_MSTR"!(ComponentPose, bAlphaBoolEnabled:bool=@n1, Translation="-10,0,-5", Rotation="45,0,0", Scale="1,1,1") -> (Pose) {
    alphaInputType=Bool
    rotationMode=BMM_Additive
    rotationSpace=BCS_BoneSpace
    translationMode=BMM_Additive
    translationSpace=BCS_BoneSpace
    alphaBoolBlend {blendInTime=0.1, blendOption=Linear, blendOutTime=0.1}
    boneToModify {boneName=dyn_jacket_mid_r_MSTR}
    rotation {pitch=45}
    translation {x=-10, z=-5}
   }
   "Transform (Modify) Bone - Bone: dyn_jacket_mid_l_MSTR"!(ComponentPose=@n2, bAlphaBoolEnabled:bool=@n1, Translation="20,0,-5", Rotation="-45,0,0", Scale="1,1,1") -> (Pose) {
    alphaInputType=Bool
    rotationMode=BMM_Additive
    rotationSpace=BCS_BoneSpace
    translationMode=BMM_Additive
    translationSpace=BCS_BoneSpace
    alphaBoolBlend {blendInTime=0.1, blendOption=Linear, blendOutTime=0.1}
    boneToModify {boneName=dyn_jacket_mid_l_MSTR}
    rotation {pitch=-45}
    translation {x=20, z=-5}
   }
   n3 = "Transform (Modify) Bone - Bone: dyn_jacket_mid_l_01"(ComponentPose=Local To Component(LocalPose=Use cached pose 'PreSim'()), Alpha=1, Translation="0,1,0", Rotation="0,5,0", Scale="1,1,1") {
    rotationMode=BMM_Additive
    rotationSpace=BCS_BoneSpace
    translationMode=BMM_Additive
    translationSpace=BCS_BoneSpace
    alphaBoolBlend {blendInTime=0.1, blendOption=Linear, blendOutTime=0.1}
    boneToModify {boneName=dyn_jacket_mid_l_01}
    rotation {yaw=5}
    translation {y=1}
   }
   n4 = "Transform (Modify) Bone - Bone: dyn_jacket_bk_l_01"(ComponentPose=@n3, Alpha=1, Translation="0,3,0", Rotation="0,15,0", Scale="1,1,1") {
    rotationMode=BMM_Additive
    rotationSpace=BCS_BoneSpace
    translationMode=BMM_Additive
    translationSpace=BCS_BoneSpace
    alphaBoolBlend {blendInTime=0.1, blendOption=Linear, blendOutTime=0.1}
    boneToModify {boneName=dyn_jacket_bk_l_01}
    rotation {yaw=15}
    translation {y=3}
   }
   n5 = "Transform (Modify) Bone - Bone: dyn_jacket_bk_r_01"(ComponentPose=@n4, Alpha=1, Translation="0,3,0", Rotation="0,15,0", Scale="1,1,1") {
    rotationMode=BMM_Additive
    rotationSpace=BCS_BoneSpace
    translationMode=BMM_Additive
    translationSpace=BCS_BoneSpace
    alphaBoolBlend {blendInTime=0.1, blendOption=Linear, blendOutTime=0.1}
    boneToModify {boneName=dyn_jacket_bk_r_01}
    rotation {yaw=15}
    translation {y=3}
   }
   n6 = "Transform (Modify) Bone - Bone: dyn_jacket_mid_r_01"(ComponentPose=@n5, Alpha=1, Translation="0,1,0", Rotation="0,5,0", Scale="1,1,1") {
    rotationMode=BMM_Additive
    rotationSpace=BCS_BoneSpace
    translationMode=BMM_Additive
    translationSpace=BCS_BoneSpace
    alphaBoolBlend {blendInTime=0.1, blendOption=Linear, blendOutTime=0.1}
    boneToModify {boneName=dyn_jacket_mid_r_01}
    rotation {yaw=5}
    translation {y=1}
   }
   n7 = RigidBody!(ComponentPose=@n6, Alpha, OverrideWorldGravity:Vector=Get GravityOverride(), ExternalForce) -> (Pose) {
    bClampLinearTranslationLimitToRefPose=true
    bOverrideWorldGravity=true
    overridePhysicsAsset=&/M_MED_Zest_Physics
    simulationSpace=BaseBoneSpace
    alphaBoolBlend {blendOption=Linear}
    baseBoneRef {boneName=spine_01}
    componentAppliedLinearAccClamp=<3000 3000 3000>
    componentLinearAccScale=<0.2 0.2 0.3>
    componentLinearVelScale=<0.2 0.2 0.3>
    overrideWorldGravity {z=-980}
   }
   n8 = Component To Local(ComponentPose=@n7)
   n9 = "Transform (Modify) Bone - Bone: dyn_jacket_mid_l_MSTR"!(
    ComponentPose=Local To Component(LocalPose=Use cached pose 'PreSim'())
    Alpha:float=Get Thigh L Rot Crouch() -> (Thigh L Rot Crouch:double)
    Translation="3,-3.5,5"
    Rotation="0,0,0"
    Scale="1,1,1"
   ) -> (Pose) {
    bAlphaBoolEnabled=false
    rotationMode=BMM_Additive
    rotationSpace=BCS_BoneSpace
    translationMode=BMM_Additive
    translationSpace=BCS_BoneSpace
    alphaBoolBlend {blendOption=Linear}
    boneToModify {boneName=dyn_jacket_mid_l_MSTR}
    translation=<3 -3.5 5>
   }
   n10 = Get Thigh_R_Rot_Crouch() -> (Thigh_R_Rot_Crouch:double)
   n11 = "Transform (Modify) Bone - Bone: dyn_jacket_mid_r_MSTR"!(ComponentPose=@n9, Alpha:float=@n10, Translation="-5,-10,7", Rotation="0,0,0", Scale="1,1,1") -> (Pose) {
    bAlphaBoolEnabled=false
    rotationMode=BMM_Additive
    rotationSpace=BCS_BoneSpace
    translationMode=BMM_Additive
    translationSpace=BCS_BoneSpace
    alphaBoolBlend {blendOption=Linear}
    boneToModify {boneName=dyn_jacket_mid_r_MSTR}
    translation=<-5 -10 7>
   }
   n12 = "Transform (Modify) Bone - Bone: dyn_jacket_fr_l_MSTR"(ComponentPose=@n11, Alpha=1, Translation="5,-4,10", Rotation="0,0,0", Scale="1,1,1") {
    bAlphaBoolEnabled=false
    rotationMode=BMM_Additive
    rotationSpace=BCS_BoneSpace
    translationMode=BMM_Additive
    translationSpace=BCS_BoneSpace
    alphaBoolBlend {blendOption=Linear}
    boneToModify {boneName=dyn_jacket_fr_l_MSTR}
    translation=<5 -4 10>
   }
   n13 = "Transform (Modify) Bone - Bone: dyn_jacket_fr_r_MSTR"(ComponentPose=@n12, Alpha=@n10, Translation="-5,0,7", Rotation="0,0,0", Scale="1,1,1") {
    bAlphaBoolEnabled=false
    rotationMode=BMM_Additive
    rotationSpace=BCS_BoneSpace
    translationMode=BMM_Additive
    translationSpace=BCS_BoneSpace
    alphaBoolBlend {blendOption=Linear}
    boneToModify {boneName=dyn_jacket_fr_r_MSTR}
    translation {x=-5, z=7}
   }
   n14 = RigidBody(ComponentPose=@n13, OverrideWorldGravity=Get GravityOverride()) {
    bClampLinearTranslationLimitToRefPose=true
    bOverrideWorldGravity=true
    overridePhysicsAsset=&/M_MED_Zest_Physics1
    simulationSpace=BaseBoneSpace
    alphaBoolBlend {blendOption=Linear}
    baseBoneRef {boneName=pelvis}
    componentAppliedLinearAccClamp=<5000 5000 5000>
    componentLinearAccScale=<0.3 0.3 0.4>
    componentLinearVelScale=<0.2 0.2 0.4>
    overrideWorldGravity {z=-980}
    simSpaceSettings {worldAlpha=0.1}
   }
   n15 = "Transform (Modify) Bone - Bone: dyn_jacket_mid_l_MSTR"(ComponentPose=@n14, Alpha=Get Thigh L Rot Crouch(), Translation="0.5,-4,1", Rotation="0,-15,0", Scale="1,1,1") {
    bAlphaBoolEnabled=false
    rotationMode=BMM_Additive
    rotationSpace=BCS_BoneSpace
    translationMode=BMM_Additive
    translationSpace=BCS_BoneSpace
    alphaBoolBlend {blendOption=Linear}
    boneToModify {boneName=dyn_jacket_mid_l_MSTR}
    rotation {yaw=-15}
    translation=<0.5 -4 1>
   }
   n16 = Component To Local(ComponentPose=@n15)
   n17 = Get bIsSkyDiving()
   n18 = Get Neck_Rot() -> (Neck_Rot:double)
   n19 = "Transform (Modify) Bone - Bone: dyn_collar_l_MSTR"(
    ComponentPose=Local To Component(LocalPose=Blend Poses by bool(BlendPose_0=@n8, BlendPose_1=@n16, bActiveValue=@n17) {blendType=Linear})
    Alpha=@n18
    Translation="2,0,-7"
    Rotation="0,-20,35"
    Scale="1,1,1"
   ) {
    rotationMode=BMM_Additive
    rotationSpace=BCS_BoneSpace
    scaleSpace=BCS_BoneSpace
    translationMode=BMM_Additive
    translationSpace=BCS_BoneSpace
    alphaBoolBlend {blendInTime=0.1, blendOption=Linear, blendOutTime=0.1}
    boneToModify {boneName=dyn_collar_l_MSTR}
    rotation {roll=35, yaw=-20}
    translation {x=2, z=-7}
   }
   n20 = "Transform (Modify) Bone - Bone: dyn_collar_r_MSTR"(ComponentPose=@n19, Alpha=@n18, Translation="-2,0,-6", Rotation="0,20,35", Scale="1,1,1") {
    rotationMode=BMM_Additive
    rotationSpace=BCS_BoneSpace
    scaleSpace=BCS_BoneSpace
    translationMode=BMM_Additive
    translationSpace=BCS_BoneSpace
    alphaBoolBlend {blendInTime=0.1, blendOption=Linear, blendOutTime=0.1}
    boneToModify {boneName=dyn_collar_r_MSTR}
    rotation {roll=35, yaw=20}
    translation {x=-2, z=-6}
   }
   n21 = "Transform (Modify) Bone - Bone: dyn_collar_r_MSTR"(ComponentPose=@n20, Alpha=Get R_Aiming_Up() -> (R_Aiming_Up:double), Translation="-1,0,0", Rotation="-15,-25,0", Scale="1,1,1") {
    rotationMode=BMM_Additive
    rotationSpace=BCS_BoneSpace
    scaleSpace=BCS_BoneSpace
    translationMode=BMM_Additive
    translationSpace=BCS_BoneSpace
    alphaBoolBlend {blendInTime=0.1, blendOption=Linear, blendOutTime=0.1}
    boneToModify {boneName=dyn_collar_r_MSTR}
    rotation {pitch=-15, yaw=-25}
    translation {x=-1}
   }
   n22 = RigidBody(ComponentPose=@n21, OverrideWorldGravity="0,0,-980") {
    bClampLinearTranslationLimitToRefPose=true
    bOverrideWorldGravity=true
    overridePhysicsAsset=&/M_MED_Zest_Physics_Collar
    simulationSpace=BaseBoneSpace
    alphaBoolBlend {blendOption=Linear}
    baseBoneRef {boneName=spine_05}
    componentAppliedLinearAccClamp=<2000 2000 2000>
    componentLinearAccScale=<0.25 0.25 0.4>
    componentLinearVelScale=<0.2 0.2 0.4>
    overrideWorldGravity {z=-980}
   }
   n23 = "Transform (Modify) Bone - Bone: dyn_collar_r_MSTR"(ComponentPose=@n22, Alpha=Get Rifle_ADS_Alpha() -> (Rifle_ADS_Alpha:double), Translation="0,0,-6", Rotation="0,-35,0", Scale="1,1,1") {
    rotationMode=BMM_Additive
    rotationSpace=BCS_BoneSpace
    scaleSpace=BCS_BoneSpace
    translationMode=BMM_Additive
    translationSpace=BCS_BoneSpace
    alphaBoolBlend {blendInTime=0.1, blendOption=Linear, blendOutTime=0.1}
    boneToModify {boneName=dyn_collar_r_MSTR}
    rotation {yaw=-35}
    translation {z=-6}
   }
   n24 = "Constraint - dyn_jacket_fr_r_01 (RR)"!(ComponentPose=@n23, Alpha=1, ConstraintWeights_0=0.6, ConstraintWeights_1:float=0.4) -> (Pose) {
    alphaBoolBlend {blendOption=Linear}
    boneToModify {boneName=dyn_jacket_fr_r_01}
    constraintSetup [{transformType=Rotation, targetBone {boneName=dyn_jacket_mid_r_01}}, {transformType=Rotation, targetBone {boneName=spine_01}}]
    constraintWeights [1,1]
   }
   n25 = "Constraint - dyn_jacket_fr_l_01 (RR)"!(ComponentPose=@n24, Alpha=1, ConstraintWeights_0=0.4, ConstraintWeights_1:float=0.6) -> (Pose) {
    alphaBoolBlend {blendOption=Linear}
    boneToModify {boneName=dyn_jacket_fr_l_01}
    constraintSetup [{transformType=Rotation, targetBone {boneName=dyn_jacket_mid_l_01}}, {transformType=Rotation, targetBone {boneName=spine_01}}]
    constraintWeights [1,1]
   }
   Output Pose(Result=Component To Local(ComponentPose=@n25)) {layerGroup=None}
   Save cached pose 'PreSim'(Pose=Copy Pose From Mesh() {bUseAttachedParent=true}) {cacheName=PreSim}
  ]}
 EventGraph {nodes [
   Event BlueprintUpdateAnimation()
   > n1 = Sequence!(execute) -> (then_0, then_1, then_2:Exec, then_3:Exec, then_4:Exec, then_5:Exec)
   > n2 = Gravity Override State Function DEPRECATED(CharacterPartAnimInstance=Self-Reference(), GameTimeInSeconds=GetGameTimeInSeconds(), GravityOverrideParams=Get Gravity Override Params())
   > Set GravityOverride(GravityOverride=@n2)
   n3 = Break Fort Anim Input Slope Info!(FortAnimInput_SlopeInfo:"Fort Anim Input Slope Info Structure (by ref)"=Get SlopeInfoInput() -> (SlopeInfoInput:Fort Anim Input Slope Info Structure)) -> (RootSlopeTranslation:Vector=~, LeftFootSlopeTranslation:Vector=~, RightFootSlopeTranslation:Vector=~, SlopeRotation:Rotator=~) {is=title K2Node_BreakStruct}
   Set slope_root_trans(execute=@n1.then_1, slope_root_trans:Vector=@n3) -> (Output_Get:Vector)
   > Set slope_foot_rot(slope_foot_rot:Rotator=@n3.SlopeRotation) -> (Output_Get:Rotator)
   n4 = ComplexSimSpaceFunction DEPRECATED!(
    execute=@n1.then_2
    CharacterPartAnimInstance:Custom Character Part Anim Instance Object Reference=Self-Reference()
    InterpSpeedMax:double=9
    PawnSpeedMax:double=20
    RInterpErrorTolerance:double=1
    DeltaTime:double=0
    DebugDraw:bool=false
    sim_space_rotation_in:Rotator=Get sim_space_rotation() -> (sim_space_rotation:Rotator)
    inComponentSpaceIn:bool=Get InComponentSpace() -> (InComponentSpace:bool)
    inComponentSpaceLastFrameIn:bool=Get InComponentSpaceLastFrame() -> (InComponentSpaceLastFrame:bool)
    lockedOnSimSpaceIn:bool=Get LockedOnSimSpace() -> (LockedOnSimSpace:bool)
    InterpSpeedIn:double=Get InterpSpeed() -> (InterpSpeed:double)
   ) -> (
    then
    sim_space_rotation:Rotator
    inComponentSpaceOut:bool
    inComponentSpaceLastFrameOut:bool
    lockedOnSimSpaceOut:bool
    InterpSpeedOut:double
   )
   > Set sim_space_rotation(sim_space_rotation:Rotator=@n4) -> (Output_Get:Rotator)
   > Set InComponentSpace(InComponentSpace:bool=@n4.inComponentSpaceOut) -> (Output_Get:bool)
   > Set InComponentSpaceLastFrame(InComponentSpaceLastFrame:bool=@n4.inComponentSpaceLastFrameOut) -> (Output_Get:bool)
   > Set LockedOnSimSpace(LockedOnSimSpace:bool=@n4.lockedOnSimSpaceOut) -> (Output_Get:bool)
   > Set InterpSpeed(InterpSpeed:double=@n4.InterpSpeedOut) -> (Output_Get:double)
   n5 = BreakTransform(InTransform=GetDeltaTransformFromRefPose(self=GetOwnerSkeletalMeshComponent(), BoneName=head, BaseName=spine_05))
   Set Neck_Rot(execute=@n1.then_4, Neck_Rot:double=MapRangeClamped(Value=BreakRotator(InRot=@n5.Rotation).Yaw, InRangeA=-5, InRangeB=22)) -> (Output_Get:double)
   n6 = BreakTransform(InTransform=GetDeltaTransformFromRefPose(self=GetOwnerSkeletalMeshComponent(), BoneName=upperArm_r, BaseName=spine_05))
   > Set R_Aiming_Up(R_Aiming_Up:double=MapRangeClamped(Value=BreakRotator(InRot=@n6.Rotation).Yaw, InRangeA=76, InRangeB=100)) -> (Output_Get:double)
   n7 = Branch(execute=@n1.then_3, Condition=OR Boolean!(A=Get bIsCrouching(), B=Get bIsJumping(), C:bool=Get bIsFalling()) -> (ReturnValue))
   n8 = BreakTransform(InTransform=GetDeltaTransformFromRefPose(self=GetOwnerSkeletalMeshComponent(), BoneName=thigh_l, BaseName=pelvis))
   > Set Thigh L Rot Crouch(Thigh L Rot Crouch:double=MapRangeClamped(Value=BreakRotator(InRot=@n8.Rotation).Yaw, InRangeA=-40, InRangeB=-100)) -> (Output_Get:double)
   n9 = BreakTransform(InTransform=GetDeltaTransformFromRefPose(self=GetOwnerSkeletalMeshComponent(), BoneName=thigh_r, BaseName=pelvis))
   > Set Thigh_R_Rot_Crouch(Thigh_R_Rot_Crouch:double=MapRangeClamped(Value=BreakRotator(InRot=@n9.Rotation).Yaw, InRangeA=-40, InRangeB=-100)) -> (Output_Get:double)
   Set Thigh L Rot Crouch(execute=@n7.else, Thigh L Rot Crouch=0)
   > Set Thigh_R_Rot_Crouch(Thigh_R_Rot_Crouch=0)
   n10 = Branch(execute=@n1.then_5, Condition=Get bIsTargeting())
   > n11 = Cast To FortPlayerAnimInstance(Object=GetAnimInstance(self=GetOwnerSkeletalMeshComponent()))
   > Set Rifle_ADS_Alpha(Rifle_ADS_Alpha:double=0) -> (Output_Get:double)
   > n12 = Switch on EFortWeaponCoreAnimation!(execute, Selection:EFortWeaponCoreAnimation Enum=Get WeaponCoreAnim(self=@n11)) -> (
    Melee:Exec
    Pistol:Exec
    Shotgun:Exec
    PaperBlueprint:Exec
    Rifle:Exec
    MeleeOneHand:Exec
    MachinePistol:Exec
    RocketLauncher:Exec
    GrenadeLauncher:Exec
    GoingCommando:Exec
    AssaultRifle:Exec
    TacticalShotgun:Exec
    SniperRifle:Exec
    TrapPlacement:Exec
    ShoulderLauncher:Exec
    AbilityDecoTool:Exec
    Crossbow:Exec
    C4:Exec
    RemoteControl:Exec
    DualWield:Exec
    AR_BullPup:Exec
    AR_ForwardGrip:Exec
    MedPackPaddles:Exec
    SMG_P90:Exec
    AR_DrumGun:Exec
    Consumable_Small:Exec
    Consumable_Large:Exec
    Balloon:Exec
    MountedTurret:Exec
    CreativeTool:Exec
    ExplosiveBow:Exec
    AshtonIndigo:Exec
    AshtonChicago:Exec
    MeleeDualWield:Exec
    Unarmed:Exec
   )
   n13 = BreakTransform(InTransform=GetDeltaTransformFromRefPose(self=GetOwnerSkeletalMeshComponent(), BoneName=head, BaseName=spine_05))
   Set Rifle_ADS_Alpha(
    execute=[@n12.AssaultRifle, @n12.Rifle, @n12.Shotgun]
    Rifle_ADS_Alpha=MapRangeClamped(Value=BreakRotator(InRot=@n13.Rotation).Yaw, InRangeA=15, InRangeB=-20)
   )
   Set Rifle_ADS_Alpha(execute=@n10.else, Rifle_ADS_Alpha=0)
   Event BlueprintInitializeAnimation()
   > Set Skel_Mesh(Skel_Mesh=GetOwnerSkeletalMeshComponent())
  ]}
]
variables [
 Gravity Override Params {
  type=GravityOverrideParamsStruct
  default {
   emoteOrMelee {gravityOverride=<0 0 -720>}
   front_End {gravityOverride=<0 0 -720>}
   onGround_Moving {gravityOverride=<0 0 -840>}
   onGround_Standing {gravityOverride=<0 0 -840>}
   skydive_DivingDown {gravityOverride=<0 0 0>}
   skydive_FloatingDown {gravityOverride=<0 0 -240>}
   skydive_GoingUp {gravityOverride=<0 -300 100>}
   skydive_Parachuting {gravityOverride=<0 0 -540>}
   swimming {gravityOverride=<0 0 0>}
  }
 }
 GravityOverride {type=Vector}
 InComponentSpace {type=boolean}
 InComponentSpaceLastFrame {type=boolean}
 InterpSpeed {type=number}
 LockedOnSimSpace {type=boolean}
 Neck_Rot {type=number}
 R_Aiming_Up {type=number}
 Rifle_ADS_Alpha {type=number}
 sim_space_rotation {type=Rotator}
 Skel_Mesh {type=SkeletalMeshComponent}
 slope_foot_rot {type=Rotator}
 slope_root_trans {type=Vector}
 Thigh L Rot Crouch {type=number}
 Thigh_R_Rot_Crouch {type=number}
]
