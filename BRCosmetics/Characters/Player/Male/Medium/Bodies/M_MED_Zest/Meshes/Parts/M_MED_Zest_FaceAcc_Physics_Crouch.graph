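# Crouch-variant physics asset for the M_MED_Zest face accessory, covering the dynamic left and right headband strands.
# Each strand is a seven-link capsule chain hanging from a master bone on dyn_headAdjuster, with tight twist and narrower second-swing limits.
# Capsules on the spine, neck, head, clavicles and upper arms provide body volumes for the strands.
# The spine_01 capsule is enlarged, offset and tilted, unlike the uniform mid-spine capsules.
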PhysicsAsset /BRCosmetics/Characters/Player/Male/Medium/Bodies/M_MED_Zest/Meshes/Parts/M_MED_Zest_FaceAcc_Physics_Crouch
bodies {
 clavicle_l [{
   shapeType=Capsule
   length=7.046
   radius=8.204
   shapeName=clavicle_l_capsule
   center=<9.6888 -0.0003 -0.195>
   rotation=<rot 89.4706 -179.998 -179.998>
  }]
 clavicle_r [{
   shapeType=Capsule
   length=7.046
   radius=8.204
   shapeName=clavicle_r_capsule
   center=<-8.379 0 0.6334>
   rotation=<rot 89.7106 -179.9981 -179.9981>
  }]
 dyn_headAdjuster [{shapeType=Capsule, length=15.8904, radius=15.6007, center=<5.3037 0 -0.0001>, rotation=<rot 90 0 0>}]
 dyn_headband_l_01 [{shapeType=Capsule, length=1.7852, radius=1.5705, shapeName=dyn_headband_l_01_capsule, center=<1.685 0 0>, rotation=<rot 90 0 0>}]
 dyn_headband_l_02 [{shapeType=Capsule, length=5.5, radius=2.5, shapeName=dyn_headband_l_02_capsule, center=<3 0 0>, rotation=<rot 90 0 0>}]
 dyn_headband_l_03 [{shapeType=Capsule, length=6, radius=2.5, shapeName=dyn_headband_l_02_capsule, center=<3 0 0>, rotation=<rot 90 0 0>}]
 dyn_headband_l_04 [{shapeType=Capsule, length=7, radius=2.5, shapeName=dyn_headband_l_02_capsule, center=<3 0 0>, rotation=<rot 90 0 0>}]
 dyn_headband_l_05 [{shapeType=Capsule, length=7, radius=2.5, shapeName=dyn_headband_l_02_capsule, center=<3 0 0>, rotation=<rot 90 0 0>}]
 dyn_headband_l_06 [{shapeType=Capsule, length=7.5, radius=2.5, shapeName=dyn_headband_l_02_capsule, center=<3 0 0>, rotation=<rot 90 0 0>}]
 dyn_headband_l_07 [{shapeType=Capsule, length=8, radius=2.5, shapeName=dyn_headband_l_02_capsule, center=<3 0 0>, rotation=<rot 90 0 0>}]
 dyn_headband_l_MSTR [{shapeType=Capsule, length=0.505, radius=0.505, shapeName=dyn_headband_l_MSTR_capsule, rotation=<rot 0 -90 0>}]
 dyn_headband_r_01 [{shapeType=Capsule, length=1.7852, radius=1.5705, shapeName=dyn_headband_l_01_capsule, center=<1.685 0 0>, rotation=<rot 90 0 0>}]
 dyn_headband_r_02 [{shapeType=Capsule, length=5.5, radius=2.25, shapeName=dyn_headband_l_02_capsule, center=<3 0 0>, rotation=<rot 90 0 0>}]
 dyn_headband_r_03 [{shapeType=Capsule, length=6, radius=2.25, shapeName=dyn_headband_l_02_capsule, center=<3 0 0>, rotation=<rot 90 0 0>}]
 dyn_headband_r_04 [{shapeType=Capsule, length=7, radius=2.25, shapeName=dyn_headband_l_02_capsule, center=<3 0 0>, rotation=<rot 90 0 0>}]
 dyn_headband_r_05 [{shapeType=Capsule, length=7, radius=2.25, shapeName=dyn_headband_l_02_capsule, center=<3 0 0>, rotation=<rot 90 0 0>}]
 dyn_headband_r_06 [{shapeType=Capsule, length=7.5, radius=2.25, shapeName=dyn_headband_l_02_capsule, center=<3 0 0>, rotation=<rot 90 0 0>}]
 dyn_headband_r_07 [{shapeType=Capsule, length=5.5, radius=2.25, shapeName=dyn_headband_l_02_capsule, center=<3 0 0>, rotation=<rot 90 0 0>}]
 dyn_headband_r_MSTR [{shapeType=Capsule, length=0.505, radius=0.505, shapeName=dyn_headband_r_MSTR_capsule, rotation=<rot 0 -90 0>}]
 neck_01 [{
   shapeType=Capsule
   length=3.6371
   radius=12.5856
   shapeName=neck_01_capsule
   center=<2.5283 2.4441 0.0186>
   rotation=<rot -89.5771 -0.0001 0>
  }]
 spine_01 [{
   shapeType=Capsule
   length=14.375
   radius=22.25
   shapeName=spine_04_capsule
   center=<-19.9572 3.7591 -0.0001>
   rotation=<rot 90 -26.565 -36.5651>
  }]
 spine_02 [{shapeType=Capsule, length=8.25, radius=15.5, shapeName=spine_04_capsule, center=<0 1.9457 0>, rotation=<rot 90 0 0>}]
 spine_03 [{shapeType=Capsule, length=8.25, radius=15.5, shapeName=spine_04_capsule, center=<0 1.9457 0>, rotation=<rot 90 0 0>}]
 spine_04 [{shapeType=Capsule, length=8.25, radius=15.5, shapeName=spine_04_capsule, center=<0 1.9457 0>, rotation=<rot 90 0 0>}]
 spine_05 [{shapeType=Capsule, length=19.5255, radius=13.9005, shapeName=spine_05_capsule, center=<-0.087 1.2997 0>, rotation=<rot 0 -0.3865 0>}]
 upperarm_l [{shapeType=Capsule, length=12, radius=8.5, shapeName=clavicle_l_capsule, center=<10 0 0>, rotation=<rot 90 0 0>}]
 upperarm_r [{shapeType=Capsule, length=12, radius=8.5, shapeName=clavicle_l_capsule, center=<-10 0 0>, rotation=<rot 90 0 0>}]
}
constraints [
 {bone1Name=neck_01, bone2Name=spine_05, swing1LimitDegrees=45, swing2LimitDegrees=45, twistLimitDegrees=45}
 {bone1Name=dyn_headAdjuster, bone2Name=neck_01, swing1LimitDegrees=45, swing2LimitDegrees=45, twistLimitDegrees=45}
 {
  bone1Name=spine_05
  bone2Name=spine_04
  swing1LimitDegrees=45
  swing1Motion=Free
  swing2LimitDegrees=45
  swing2Motion=Free
  twistLimitDegrees=45
  twistMotion=Free
 }
 {
  bone1Name=dyn_headband_l_MSTR
  bone2Name=dyn_headAdjuster
  swing1LimitDegrees=45
  swing1Motion=Free
  swing2LimitDegrees=45
  swing2Motion=Free
  twistLimitDegrees=45
  twistMotion=Free
 }
 {
  bone1Name=dyn_headband_r_MSTR
  bone2Name=dyn_headAdjuster
  swing1LimitDegrees=45
  swing1Motion=Free
  swing2LimitDegrees=45
  swing2Motion=Free
  twistLimitDegrees=45
  twistMotion=Free
 }
 {bone1Name=dyn_headband_l_01, bone2Name=dyn_headband_l_MSTR, swing1LimitDegrees=45, swing2LimitDegrees=10, twistLimitDegrees=5}
 {bone1Name=dyn_headband_l_02, bone2Name=dyn_headband_l_01, swing1LimitDegrees=45, swing2LimitDegrees=20, twistLimitDegrees=5}
 {bone1Name=dyn_headband_l_03, bone2Name=dyn_headband_l_02, swing1LimitDegrees=45, swing2LimitDegrees=20, twistLimitDegrees=5}
 {bone1Name=dyn_headband_l_04, bone2Name=dyn_headband_l_03, swing1LimitDegrees=45, swing2LimitDegrees=20, twistLimitDegrees=5}
 {bone1Name=dyn_headband_l_05, bone2Name=dyn_headband_l_04, swing1LimitDegrees=45, swing2LimitDegrees=20, twistLimitDegrees=5}
 {bone1Name=dyn_headband_l_06, bone2Name=dyn_headband_l_05, swing1LimitDegrees=45, swing2LimitDegrees=20, twistLimitDegrees=5}
 {bone1Name=dyn_headband_l_07, bone2Name=dyn_headband_l_06, swing1LimitDegrees=45, swing2LimitDegrees=20, twistLimitDegrees=5}
 {bone1Name=dyn_headband_r_01, bone2Name=dyn_headband_r_MSTR, swing1LimitDegrees=45, swing2LimitDegrees=10, twistLimitDegrees=5}
 {bone1Name=dyn_headband_r_02, bone2Name=dyn_headband_r_01, swing1LimitDegrees=45, swing2LimitDegrees=20, twistLimitDegrees=5}
 {bone1Name=dyn_headband_r_03, bone2Name=dyn_headband_r_02, swing1LimitDegrees=45, swing2LimitDegrees=20, twistLimitDegrees=5}
 {bone1Name=dyn_headband_r_04, bone2Name=dyn_headband_r_03, swing1LimitDegrees=45, swing2LimitDegrees=20, twistLimitDegrees=5}
 {bone1Name=dyn_headband_r_05, bone2Name=dyn_headband_r_04, swing1LimitDegrees=45, swing2LimitDegrees=20, twistLimitDegrees=5}
 {bone1Name=dyn_headband_r_06, bone2Name=dyn_headband_r_05, swing1LimitDegrees=45, swing2LimitDegrees=20, twistLimitDegrees=5}
 {bone1Name=dyn_headband_r_07, bone2Name=dyn_headband_r_06, swing1LimitDegrees=45, swing2LimitDegrees=20, twistLimitDegrees=5}
 {
  bone1Name=clavicle_r
  bone2Name=spine_05
  swing1LimitDegrees=45
  swing1Motion=Free
  swing2LimitDegrees=45
  swing2Motion=Free
  twistLimitDegrees=45
  twistMotion=Free
 }
 {
  bone1Name=clavicle_l
  bone2Name=spine_05
  swing1LimitDegrees=45
  swing1Motion=Free
  swing2LimitDegrees=45
  swing2Motion=Free
  twistLimitDegrees=45
  twistMotion=Free
 }
 {
  bone1Name=upperarm_l
  bone2Name=clavicle_l
  swing1LimitDegrees=45
  swing1Motion=Free
  swing2LimitDegrees=45
  swing2Motion=Free
  twistLimitDegrees=45
  twistMotion=Free
 }
 {
  bone1Name=upperarm_r
  bone2Name=clavicle_r
  swing1LimitDegrees=45
  swing1Motion=Free
  swing2LimitDegrees=45
  swing2Motion=Free
  twistLimitDegrees=45
  twistMotion=Free
 }
 {
  bone1Name=spine_04
  bone2Name=spine_03
  swing1LimitDegrees=45
  swing1Motion=Free
  swing2LimitDegrees=45
  swing2Motion=Free
  twistLimitDegrees=45
  twistMotion=Free
 }
 {
  bone1Name=spine_03
  bone2Name=spine_02
  swing1LimitDegrees=45
  swing1Motion=Free
  swing2LimitDegrees=45
  swing2Motion=Free
  twistLimitDegrees=45
  twistMotion=Free
 }
 {
  bone1Name=spine_02
  bone2Name=spine_01
  swing1LimitDegrees=45
  swing1Motion=Free
  swing2LimitDegrees=45
  swing2Motion=Free
  twistLimitDegrees=45
  twistMotion=Free
 }
]
