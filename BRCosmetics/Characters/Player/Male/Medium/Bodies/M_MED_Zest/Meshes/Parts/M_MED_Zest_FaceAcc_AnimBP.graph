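# Post-copy anim blueprint for the Zest character's face accessory, animating headband tails and hair over the pose copied from the parent mesh.
# Backbling poses headband procedurally with Look At chains and bone offsets, varying by length; otherwise it uses standing/crouch RigidBody, or a procedural fallback when physics is disabled.
# Hair Anim Dynamics runs last.
# The event graph derives gravity, backbling state and correction alphas from head, neck and spine rotations.
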AnimBlueprint /BRCosmetics/Characters/Player/Male/Medium/Bodies/M_MED_Zest/Meshes/Parts/M_MED_Zest_FaceAcc_AnimBP
target_skeleton=&/M_MED_Zest_FaceAcc_Skeleton
graphs [
 AnimGraph {nodes [
   Layered blend per bone!(
    BasePose=Use cached pose 'HeadRotation_Cache'()
    BlendPoses_0
    BlendPoses_1:Pose Link Structure=Use cached pose 'HeadRotation_Cache'()
    BlendWeights_0=0.7
    BlendWeights_1:float=1
   ) -> (Pose) {
    blendPoses [{linkId=-1, sourceLinkId=-1}, {linkId=-1, sourceLinkId=-1}]
    blendWeights [1,1]
    layerSetup [{branchFilters [{blendDepth=1, boneName=neck_01}]}, {branchFilters [{blendDepth=1, boneName=faceAttach}]}]
   }
   "Transform (Modify) Bone - Bone: dyn_headband_MSTR"(Alpha=Get Head_Down_Alpha() -> (Head_Down_Alpha:double), Translation="0,0,0", Rotation="0,0,6", Scale="1,1,1") {
    rotationMode=BMM_Additive
    rotationSpace=BCS_BoneSpace
    translationMode=BMM_Additive
    translationSpace=BCS_BoneSpace
    alphaBoolBlend {blendOption=Linear}
    boneToModify {boneName=dyn_headband_MSTR}
    rotation {roll=6}
   }
   n1 = RigidBody!(
    ComponentPose=Local To Component(LocalPose=Use cached pose 'HeadRotation_Cache'())
    Alpha
    OverrideWorldGravity:Vector=Get GravityOverride()
    ExternalForce
   ) -> (Pose) {
    bClampLinearTranslationLimitToRefPose=true
    bOverrideWorldGravity=true
    overridePhysicsAsset=&/M_MED_Zest_FaceAcc_Physics_Crouch
    simulationSpace=BaseBoneSpace
    alphaBoolBlend {blendOption=Linear}
    baseBoneRef {boneName=neck_01}
    componentAppliedLinearAccClamp=<3000 3000 3000>
    componentLinearAccScale=<0.2 0.2 0.3>
    componentLinearVelScale=<0.2 0.2 0.3>
    overrideWorldGravity {z=-980}
    simSpaceSettings {
     maxAngularAcceleration=1000
     maxAngularVelocity=3
     maxLinearAcceleration=700
     maxLinearVelocity=500
     velocityScaleZ=0
     worldAlpha=1
     externalLinearDragV=<0.2 0.2>
    }
   }
   n2 = RigidBody(ComponentPose=Local To Component(LocalPose=Use cached pose 'HeadRotation_Cache'()), OverrideWorldGravity=Get GravityOverride()) {
    bClampLinearTranslationLimitToRefPose=true
    bOverrideWorldGravity=true
    overridePhysicsAsset=&/M_MED_Zest_FaceAcc_Physics
    simulationSpace=BaseBoneSpace
    alphaBoolBlend {blendOption=Linear}
    baseBoneRef {boneName=neck_01}
    componentAppliedLinearAccClamp=<3000 3000 3000>
    componentLinearAccScale=<0.2 0.2 0.3>
    componentLinearVelScale=<0.2 0.2 0.3>
    overrideWorldGravity {z=-980}
    simSpaceSettings {
     maxAngularAcceleration=1000
     maxAngularVelocity=3
     maxLinearAcceleration=700
     maxLinearVelocity=500
     velocityScaleZ=0
     worldAlpha=1
     externalLinearDragV=<0.2 0.2>
    }
   }
   n3 = "Transform (Modify) Bone - Bone: dyn_headband_MSTR"(
    ComponentPose=Local To Component(LocalPose=Use cached pose 'HeadRotation_Cache'())
    Alpha=1
    Translation="0,0,0"
    Rotation="0,0,-7"
    Scale="1,1,1"
   ) {
    rotationMode=BMM_Additive
    rotationSpace=BCS_BoneSpace
    translationMode=BMM_Additive
    translationSpace=BCS_BoneSpace
    alphaBoolBlend {blendOption=Linear}
    boneToModify {boneName=dyn_headband_MSTR}
    rotation {roll=-7}
   }
   n4 = "Transform (Modify) Bone - Bone: dyn_headband_MSTR"(ComponentPose=@n3, Alpha=Get Head_Left_Alpha() -> (Head_Left_Alpha:double), Translation="0,0,0", Rotation="10,25,10", Scale="1,1,1") {
    rotationMode=BMM_Additive
    rotationSpace=BCS_BoneSpace
    translationMode=BMM_Additive
    translationSpace=BCS_BoneSpace
    alphaBoolBlend {blendOption=Linear}
    boneToModify {boneName=dyn_headband_MSTR}
    rotation=<rot 10 25 10>
   }
   n5 = "Transform (Modify) Bone - Bone: dyn_headband_l_02"(ComponentPose=@n4, Alpha=1, Translation="0,0,0", Rotation="0,-15,0", Scale="1,1,1") {
    rotationMode=BMM_Additive
    rotationSpace=BCS_BoneSpace
    translationMode=BMM_Additive
    translationSpace=BCS_BoneSpace
    alphaBoolBlend {blendOption=Linear}
    boneToModify {boneName=dyn_headband_l_02}
    rotation {yaw=-15}
   }
   n6 = "Transform (Modify) Bone - Bone: dyn_headband_r_02"(ComponentPose=@n5, Alpha=1, Translation="0,0,0", Rotation="0,-15,0", Scale="1,1,1") {
    rotationMode=BMM_Additive
    rotationSpace=BCS_BoneSpace
    translationMode=BMM_Additive
    translationSpace=BCS_BoneSpace
    alphaBoolBlend {blendOption=Linear}
    boneToModify {boneName=dyn_headband_r_02}
    rotation {yaw=-15}
   }
   n7 = Look At - Bone: dyn_headband_l_02(ComponentPose=@n6, Alpha=Get Head_Crouch_Alpha() -> (Head_Crouch_Alpha:double)) {
    alphaBoolBlend {blendOption=Linear}
    boneToModify {boneName=dyn_headband_l_02}
    lookAtLocation=<0 -14>
    lookAtTarget {boneReference {boneName=spine_04}}
    lookAt_Axis {axis=<1 0>}
   }
   n8 = Look At - Bone: dyn_headband_l_03(ComponentPose=@n7, Alpha=0.4) {
    alphaBoolBlend {blendOption=Linear}
    boneToModify {boneName=dyn_headband_l_03}
    lookAtLocation=<0 -20 -1>
    lookAtTarget {boneReference {boneName=spine_03}}
    lookAt_Axis {axis=<1 0>}
   }
   n9 = Look At - Bone: dyn_headband_l_04(ComponentPose=@n8, Alpha=0.6) {
    alphaBoolBlend {blendOption=Linear}
    boneToModify {boneName=dyn_headband_l_04}
    lookAtLocation=<0 -16 -3>
    lookAtTarget {boneReference {boneName=spine_02}}
    lookAt_Axis {axis=<1 0>}
   }
   n10 = Look At - Bone: dyn_headband_l_05(ComponentPose=@n9, Alpha=0.8) {
    alphaBoolBlend {blendOption=Linear}
    boneToModify {boneName=dyn_headband_l_05}
    lookAtLocation=<0 -18 -5>
    lookAtTarget {boneReference {boneName=spine_01}}
    lookAt_Axis {axis=<1 0>}
   }
   n11 = Look At - Bone: dyn_headband_l_06(ComponentPose=@n10, Alpha=1) {
    alphaBoolBlend {blendOption=Linear}
    boneToModify {boneName=dyn_headband_l_06}
    lookAtLocation=<0 -23 -7>
    lookAtTarget {boneReference {boneName=pelvis}}
    lookAt_Axis {axis=<1 0>}
   }
   n12 = Anim Dynamics - Bone: dyn_headband_l_06(ComponentPose=@n11, Alpha=Get Backbling Alpha() -> (Backbling Alpha:double), ExternalForce="0,0,0") {
    angularDampingOverride=0.9999
    angularSpringConstant=100
    bAngularSpring=true
    bOverrideAngularDamping=true
    bUseGravityOverride=true
    alphaBoolBlend {blendOption=Linear}
    boundBone {boneName=dyn_headband_l_06}
    componentAppliedLinearAccClamp=<3000 3000 3000>
    componentLinearAccScale=<0.25 0.25 0.4>
    componentLinearVelScale=<0.25 0.25 0.4>
    gravityOverride {z=-400}
    physicsBodyDefinitions [{
      sphereCollisionRadius=0
      boundBone {boneName=dyn_headband_l_06}
      boxExtents=<25 10 10>
      constraintSetup {angularLimitsMax=<1 5 1>, angularLimitsMin=<-1 -5 -1>, angularTarget=<0 0 0>}
      localJointOffset=<10 0 0>
     }]
    relativeSpaceBone {boneName=dyn_headband_l_05}
   }
   n13 = Look At - Bone: dyn_headband_r_02(ComponentPose=@n12, Alpha=Get Head_Crouch_Alpha()) {
    alphaBoolBlend {blendOption=Linear}
    boneToModify {boneName=dyn_headband_r_02}
    lookAtLocation=<0 -15>
    lookAtTarget {boneReference {boneName=spine_04}}
    lookAt_Axis {axis=<1 0>}
   }
   n14 = Look At - Bone: dyn_headband_r_03(ComponentPose=@n13, Alpha=0.4) {
    alphaBoolBlend {blendOption=Linear}
    boneToModify {boneName=dyn_headband_r_03}
    lookAtLocation=<0 -21 1>
    lookAtTarget {boneReference {boneName=spine_03}}
    lookAt_Axis {axis=<1 0>}
   }
   n15 = Look At - Bone: dyn_headband_r_04(ComponentPose=@n14, Alpha=0.6) {
    alphaBoolBlend {blendOption=Linear}
    boneToModify {boneName=dyn_headband_r_04}
    lookAtLocation=<0 -17 3>
    lookAtTarget {boneReference {boneName=spine_02}}
    lookAt_Axis {axis=<1 0>}
   }
   n16 = Look At - Bone: dyn_headband_r_05(ComponentPose=@n15, Alpha=0.8) {
    alphaBoolBlend {blendOption=Linear}
    boneToModify {boneName=dyn_headband_r_05}
    lookAtLocation=<0 -20 5>
    lookAtTarget {boneReference {boneName=spine_01}}
    lookAt_Axis {axis=<1 0>}
   }
   n17 = Look At - Bone: dyn_headband_r_06(ComponentPose=@n16, Alpha=1) {
    alphaBoolBlend {blendOption=Linear}
    boneToModify {boneName=dyn_headband_r_06}
    lookAtLocation=<0 -23 6>
    lookAtTarget {boneReference {boneName=pelvis}}
    lookAt_Axis {axis=<1 0>}
   }
   n18 = Anim Dynamics - Bone: dyn_headband_r_06(ComponentPose=@n17, Alpha=Get Backbling Alpha(), ExternalForce="0,0,0") {
    angularDampingOverride=0.9999
    angularSpringConstant=100
    bAngularSpring=true
    bOverrideAngularDamping=true
    bUseGravityOverride=true
    alphaBoolBlend {blendOption=Linear}
    boundBone {boneName=dyn_headband_r_06}
    componentAppliedLinearAccClamp=<3000 3000 3000>
    componentLinearAccScale=<0.25 0.25 0.4>
    componentLinearVelScale=<0.25 0.25 0.4>
    gravityOverride {z=-400}
    physicsBodyDefinitions [{
      sphereCollisionRadius=0
      boundBone {boneName=dyn_headband_r_06}
      boxExtents=<25 10 10>
      constraintSetup {angularLimitsMax=<1 3 1>, angularLimitsMin=<-1 -12 -1>, angularTarget=<0 0 0>}
      localJointOffset=<10 0 0>
     }]
    relativeSpaceBone {boneName=dyn_headband_r_05}
   }
   n19 = "Transform (Modify) Bone - Bone: dyn_headband_MSTR"!(ComponentPose=@n18, bAlphaBoolEnabled:bool=Get bIsSkyDiving(), Translation="0,0,0", Rotation="0,0,15", Scale="1,1,1") -> (Pose) {
    alphaInputType=Bool
    rotationMode=BMM_Additive
    rotationSpace=BCS_BoneSpace
    translationMode=BMM_Additive
    translationSpace=BCS_BoneSpace
    alphaBoolBlend {blendOption=Linear}
    boneToModify {boneName=dyn_headband_MSTR}
    rotation {roll=15}
   }
   n20 = Blend Poses by bool(
    BlendPose_0=Component To Local(ComponentPose=@n19)
    BlendPose_1=Blend Poses by bool(BlendPose_0=Component To Local(ComponentPose=@n1), BlendPose_1=Component To Local(ComponentPose=@n2), bActiveValue=Get bIsCrouching()) {blendType=Linear}
    bActiveValue=NOT Boolean(A=Get bRigidBodyEnabled())
   ) {blendType=Linear}
   n21 = "Transform (Modify) Bone - Bone: dyn_headband_MSTR"!(
    ComponentPose=Local To Component(LocalPose=Use cached pose 'HeadRotation_Cache'())
    Alpha:float=1
    Translation="0,0,0"
    Rotation="0,0,-25"
    Scale="1,1,1"
   ) -> (Pose) {
    rotationMode=BMM_Additive
    rotationSpace=BCS_BoneSpace
    translationMode=BMM_Additive
    translationSpace=BCS_BoneSpace
    alphaBoolBlend {blendOption=Linear}
    boneToModify {boneName=dyn_headband_MSTR}
    rotation {roll=-25}
   }
   n22 = Get Head_Up_Alpha() -> (Head_Up_Alpha:double)
   n23 = "Transform (Modify) Bone - Bone: dyn_headband_r_03"(ComponentPose=@n21, Alpha=@n22, Translation="0,0,0", Rotation="0,5,0", Scale="1,1,1") {
    rotationMode=BMM_Additive
    rotationSpace=BCS_BoneSpace
    translationMode=BMM_Additive
    translationSpace=BCS_BoneSpace
    alphaBoolBlend {blendOption=Linear}
    boneToModify {boneName=dyn_headband_r_03}
    rotation {yaw=5}
   }
   n24 = "Transform (Modify) Bone - Bone: dyn_headband_MSTR"(ComponentPose=@n23, Alpha=Get Head_Left_Alpha(), Translation="0,0,0", Rotation="10,25,2", Scale="1,1,1") {
    rotationMode=BMM_Additive
    rotationSpace=BCS_BoneSpace
    translationMode=BMM_Additive
    translationSpace=BCS_BoneSpace
    alphaBoolBlend {blendOption=Linear}
    boneToModify {boneName=dyn_headband_MSTR}
    rotation=<rot 10 25 2>
   }
   n25 = Look At - Bone: dyn_headband_r_03(ComponentPose=@n24, Alpha=@n22) {
    alphaBoolBlend {blendOption=Linear}
    alphaScaleBiasClamp {bClampResult=true, clampMin=0.3}
    boneToModify {boneName=dyn_headband_r_03}
    lookAtLocation=<0 -25>
    lookAtTarget {boneReference {boneName=spine_03}}
    lookAt_Axis {axis=<1 0>}
   }
   n26 = Look At - Bone: dyn_headband_r_04(ComponentPose=@n25, Alpha=1) {
    alphaBoolBlend {blendOption=Linear}
    boneToModify {boneName=dyn_headband_r_04}
    lookAtLocation=<0 -12>
    lookAtTarget {boneReference {boneName=spine_04}}
    lookAt_Axis {axis=<1 0>}
   }
   n27 = Look At - Bone: dyn_headband_r_05(ComponentPose=@n26, Alpha=1) {
    alphaBoolBlend {blendOption=Linear}
    boneToModify {boneName=dyn_headband_r_05}
    lookAtLocation=<0 -20>
    lookAtTarget {boneReference {boneName=spine_01}}
    lookAt_Axis {axis=<1 0>}
   }
   n28 = Look At - Bone: dyn_headband_l_03(ComponentPose=@n27, Alpha=@n22) {
    alphaBoolBlend {blendOption=Linear}
    alphaScaleBiasClamp {bClampResult=true, clampMin=0.3}
    boneToModify {boneName=dyn_headband_l_03}
    lookAtLocation=<0 -25>
    lookAtTarget {boneReference {boneName=spine_03}}
    lookAt_Axis {axis=<1 0>}
   }
   n29 = Look At - Bone: dyn_headband_l_04(ComponentPose=@n28, Alpha=1) {
    alphaBoolBlend {blendOption=Linear}
    boneToModify {boneName=dyn_headband_l_04}
    lookAtLocation=<0 -12>
    lookAtTarget {boneReference {boneName=spine_04}}
    lookAt_Axis {axis=<1 0>}
   }
   n30 = Look At - Bone: dyn_headband_l_05(ComponentPose=@n29, Alpha=1) {
    alphaBoolBlend {blendOption=Linear}
    boneToModify {boneName=dyn_headband_l_05}
    lookAtLocation=<0 -20>
    lookAtTarget {boneReference {boneName=spine_01}}
    lookAt_Axis {axis=<1 0>}
   }
   n31 = "Transform (Modify) Bone - Bone: dyn_headband_MSTR"(
    ComponentPose=Local To Component(LocalPose=Use cached pose 'HeadRotation_Cache'())
    Alpha=1
    Translation="0,0,0"
    Rotation="0,0,-15"
    Scale="1,1,1"
   ) {
    rotationMode=BMM_Additive
    rotationSpace=BCS_BoneSpace
    translationMode=BMM_Additive
    translationSpace=BCS_BoneSpace
    alphaBoolBlend {blendOption=Linear}
    boneToModify {boneName=dyn_headband_MSTR}
    rotation {roll=-15}
   }
   n32 = "Transform (Modify) Bone - Bone: dyn_headband_MSTR"(ComponentPose=@n31, Alpha=Get Head_Up_Alpha(), Translation="0,0,0", Rotation="0,0,10", Scale="1,1,1") {
    rotationMode=BMM_Additive
    rotationSpace=BCS_BoneSpace
    translationMode=BMM_Additive
    translationSpace=BCS_BoneSpace
    alphaBoolBlend {blendOption=Linear}
    boneToModify {boneName=dyn_headband_MSTR}
    rotation {roll=10}
   }
   n33 = "Transform (Modify) Bone - Bone: dyn_headband_MSTR"(ComponentPose=@n32, Alpha=Get Head_Down_Alpha(), Translation="0,0,0", Rotation="0,0,6", Scale="1,1,1") {
    rotationMode=BMM_Additive
    rotationSpace=BCS_BoneSpace
    translationMode=BMM_Additive
    translationSpace=BCS_BoneSpace
    alphaBoolBlend {blendOption=Linear}
    boneToModify {boneName=dyn_headband_MSTR}
    rotation {roll=6}
   }
   n34 = "Transform (Modify) Bone - Bone: dyn_headband_MSTR"(ComponentPose=@n33, Alpha=Get Head_Left_Alpha(), Translation="0,0,0", Rotation="20,20,10", Scale="1,1,1") {
    rotationMode=BMM_Additive
    rotationSpace=BCS_BoneSpace
    translationMode=BMM_Additive
    translationSpace=BCS_BoneSpace
    alphaBoolBlend {blendOption=Linear}
    boneToModify {boneName=dyn_headband_MSTR}
    rotation=<rot 20 20 10>
   }
   n35 = "Transform (Modify) Bone - Bone: dyn_headband_r_02"(ComponentPose=@n34, Alpha=1, Translation="0,0,0", Rotation="0,-15,0", Scale="1,1,1") {
    rotationMode=BMM_Additive
    rotationSpace=BCS_BoneSpace
    translationMode=BMM_Additive
    translationSpace=BCS_BoneSpace
    alphaBoolBlend {blendOption=Linear}
    boneToModify {boneName=dyn_headband_r_02}
    rotation {yaw=-15}
   }
   n36 = "Transform (Modify) Bone - Bone: dyn_headband_l_02"(ComponentPose=@n35, Alpha=1, Translation="0,0,0", Rotation="0,-15,0", Scale="1,1,1") {
    rotationMode=BMM_Additive
    rotationSpace=BCS_BoneSpace
    translationMode=BMM_Additive
    translationSpace=BCS_BoneSpace
    alphaBoolBlend {blendOption=Linear}
    boneToModify {boneName=dyn_headband_l_02}
    rotation {yaw=-15}
   }
   n37 = Look At - Bone: dyn_headband_l_02(ComponentPose=@n36, Alpha=Get Head_Crouch_Alpha()) {
    alphaBoolBlend {blendOption=Linear}
    boneToModify {boneName=dyn_headband_l_02}
    lookAtLocation=<0 -16>
    lookAtTarget {boneReference {boneName=spine_04}}
    lookAt_Axis {axis=<1 0>}
   }
   n38 = Look At - Bone: dyn_headband_l_03(ComponentPose=@n37, Alpha=0.4) {
    alphaBoolBlend {blendOption=Linear}
    boneToModify {boneName=dyn_headband_l_03}
    lookAtLocation=<0 -18 -2>
    lookAtTarget {boneReference {boneName=spine_03}}
    lookAt_Axis {axis=<1 0>}
   }
   n39 = Look At - Bone: dyn_headband_l_04(ComponentPose=@n38, Alpha=0.6) {
    alphaBoolBlend {blendOption=Linear}
    boneToModify {boneName=dyn_headband_l_04}
    lookAtLocation=<0 -20 -6>
    lookAtTarget {boneReference {boneName=spine_02}}
    lookAt_Axis {axis=<1 0>}
   }
   n40 = Look At - Bone: dyn_headband_l_05(ComponentPose=@n39, Alpha=0.8) {
    alphaBoolBlend {blendOption=Linear}
    boneToModify {boneName=dyn_headband_l_05}
    lookAtLocation=<0 -18 -8>
    lookAtTarget {boneReference {boneName=spine_01}}
    lookAt_Axis {axis=<1 0>}
   }
   n41 = Look At - Bone: dyn_headband_l_06(ComponentPose=@n40, Alpha=SelectFloat(A=0.7, B=1, bPickA=Get Is Backbling Long() -> (Is Backbling Long:bool))) {
    alphaBoolBlend {blendOption=Linear}
    boneToModify {boneName=dyn_headband_l_06}
    lookAtLocation=<0 -24 -10>
    lookAtTarget {boneReference {boneName=pelvis}}
    lookAt_Axis {axis=<1 0>}
   }
   n42 = Anim Dynamics - Bone: dyn_headband_l_06(ComponentPose=@n41, Alpha=Get Backbling Alpha(), ExternalForce="0,0,0") {
    angularDampingOverride=0.9999
    angularSpringConstant=100
    bAngularSpring=true
    bOverrideAngularDamping=true
    bUseGravityOverride=true
    alphaBoolBlend {blendOption=Linear}
    boundBone {boneName=dyn_headband_l_06}
    componentAppliedLinearAccClamp=<3000 3000 3000>
    componentLinearAccScale=<0.25 0.25 0.3>
    componentLinearVelScale=<0.25 0.25 0.3>
    gravityOverride {z=-400}
    physicsBodyDefinitions [{
      sphereCollisionRadius=0
      boundBone {boneName=dyn_headband_l_06}
      boxExtents=<35 15 15>
      constraintSetup {angularLimitsMax=<0 8 0>, angularLimitsMin=<0 -8 0>, angularTarget=<0 0 0>}
      localJointOffset=<8 0 0>
     }]
    relativeSpaceBone {boneName=dyn_headband_l_05}
   }
   n43 = Look At - Bone: dyn_headband_r_02(ComponentPose=@n42, Alpha=Get Head_Crouch_Alpha()) {
    alphaBoolBlend {blendOption=Linear}
    boneToModify {boneName=dyn_headband_r_02}
    lookAtLocation=<0 -16>
    lookAtTarget {boneReference {boneName=spine_04}}
    lookAt_Axis {axis=<1 0>}
   }
   n44 = Look At - Bone: dyn_headband_r_03(ComponentPose=@n43, Alpha=0.4) {
    alphaBoolBlend {blendOption=Linear}
    boneToModify {boneName=dyn_headband_r_03}
    lookAtLocation=<0 -19 2>
    lookAtTarget {boneReference {boneName=spine_03}}
    lookAt_Axis {axis=<1 0>}
   }
   n45 = Look At - Bone: dyn_headband_r_04(ComponentPose=@n44, Alpha=0.6) {
    alphaBoolBlend {blendOption=Linear}
    boneToModify {boneName=dyn_headband_r_04}
    lookAtLocation=<0 -20 4>
    lookAtTarget {boneReference {boneName=spine_02}}
    lookAt_Axis {axis=<1 0>}
   }
   n46 = Look At - Bone: dyn_headband_r_05(ComponentPose=@n45, Alpha=0.8) {
    alphaBoolBlend {blendOption=Linear}
    boneToModify {boneName=dyn_headband_r_05}
    lookAtLocation=<0 -18 5>
    lookAtTarget {boneReference {boneName=spine_01}}
    lookAt_Axis {axis=<1 0>}
   }
   n47 = Look At - Bone: dyn_headband_r_06(ComponentPose=@n46, Alpha=SelectFloat(A=0.7, B=1, bPickA=Get Is Backbling Long())) {
    alphaBoolBlend {blendOption=Linear}
    boneToModify {boneName=dyn_headband_r_06}
    lookAtLocation=<0 -25 6>
    lookAtTarget {boneReference {boneName=pelvis}}
    lookAt_Axis {axis=<1 0>}
   }
   n48 = Anim Dynamics - Bone: dyn_headband_r_06(ComponentPose=@n47, Alpha=Get Backbling Alpha(), ExternalForce="0,0,0") {
    angularDampingOverride=0.9999
    angularSpringConstant=100
    bAngularSpring=true
    bOverrideAngularDamping=true
    bUseGravityOverride=true
    alphaBoolBlend {blendOption=Linear}
    boundBone {boneName=dyn_headband_r_06}
    componentAppliedLinearAccClamp=<3000 3000 3000>
    componentLinearAccScale=<0.25 0.25 0.3>
    componentLinearVelScale=<0.25 0.25 0.3>
    gravityOverride {z=-400}
    physicsBodyDefinitions [{
      sphereCollisionRadius=0
      boundBone {boneName=dyn_headband_r_06}
      boxExtents=<35 15 15>
      constraintSetup {angularLimitsMax=<0 7 0>, angularLimitsMin=<0 -7 0>, angularTarget=<0 0 0>}
      localJointOffset=<8 0 0>
     }]
    relativeSpaceBone {boneName=dyn_headband_r_05}
   }
   n49 = "Transform (Modify) Bone - Bone: dyn_headband_MSTR"(ComponentPose=@n48, Alpha=Get Spine_Rot() -> (Spine_Rot:double), Translation="0,0,0", Rotation="0,0,5", Scale="1,1,1") {
    rotationMode=BMM_Additive
    rotationSpace=BCS_BoneSpace
    translationMode=BMM_Additive
    translationSpace=BCS_BoneSpace
    alphaBoolBlend {blendOption=Linear}
    boneToModify {boneName=dyn_headband_MSTR}
    rotation {roll=5}
   }
   n50 = Blend Poses by bool(BlendPose_0=Component To Local(ComponentPose=@n30), BlendPose_1=Component To Local(ComponentPose=@n49), bActiveValue=Get Is Backbling Long()) {blendType=Linear}
   n51 = Get isBackBling() -> (isBackBling:bool)
   n52 = Anim Dynamics - Bone: dyn_hair_r_1(
    ComponentPose=Local To Component(LocalPose=Blend Poses by bool(BlendPose_0=@n50, BlendPose_1=@n20, bActiveValue=@n51) {blendType=Linear})
    Alpha=0.6
    ExternalForce="0,0,0"
   ) {
    angularDampingOverride=0.9999
    angularSpringConstant=6000
    bAngularSpring=true
    bOverrideAngularDamping=true
    bUseGravityOverride=true
    alphaBoolBlend {blendOption=Linear}
    boundBone {boneName=dyn_hair_r_1}
    componentAppliedLinearAccClamp=<3000 3000 3000>
    componentLinearAccScale=<0.25 0.25 0.4>
    componentLinearVelScale=<0.25 0.25 0.4>
    gravityOverride {z=-400}
    physicsBodyDefinitions [{
      sphereCollisionRadius=0
      boundBone {boneName=dyn_hair_r_1}
      boxExtents=<25 10 10>
      constraintSetup {angularLimitsMax=<1 5 5>, angularLimitsMin=<-1 -25 -5>, angularTarget=<0 -20 0>}
      localJointOffset=<5 0 0>
     }]
    relativeSpaceBone {boneName=dyn_headAdjuster}
   }
   n53 = Anim Dynamics - Bone: dyn_hair_m_1(ComponentPose=@n52, Alpha=0.6, ExternalForce="0,0,0") {
    angularDampingOverride=0.9999
    angularSpringConstant=6000
    bAngularSpring=true
    bOverrideAngularDamping=true
    bUseGravityOverride=true
    alphaBoolBlend {blendOption=Linear}
    boundBone {boneName=dyn_hair_m_1}
    componentAppliedLinearAccClamp=<3000 3000 3000>
    componentLinearAccScale=<0.25 0.25 0.4>
    componentLinearVelScale=<0.25 0.25 0.4>
    gravityOverride {z=-400}
    physicsBodyDefinitions [{
      sphereCollisionRadius=0
      boundBone {boneName=dyn_hair_m_1}
      boxExtents=<20 10 10>
      constraintSetup {angularLimitsMax=<1 5 7.5>, angularLimitsMin=<-1 -15 -7.5>, angularTarget=<0 -15 0>}
      localJointOffset=<4 0 0>
     }]
    relativeSpaceBone {boneName=dyn_headAdjuster}
   }
   n54 = Anim Dynamics - Bone: dyn_hair_l_1(ComponentPose=@n53, Alpha=0.6, ExternalForce="0,0,0") {
    angularDampingOverride=0.9999
    angularSpringConstant=6000
    bAngularSpring=true
    bOverrideAngularDamping=true
    bUseGravityOverride=true
    alphaBoolBlend {blendOption=Linear}
    boundBone {boneName=dyn_hair_l_1}
    componentAppliedLinearAccClamp=<3000 3000 3000>
    componentLinearAccScale=<0.25 0.25 0.4>
    componentLinearVelScale=<0.25 0.25 0.4>
    gravityOverride {z=-400}
    physicsBodyDefinitions [{
      sphereCollisionRadius=0
      boundBone {boneName=dyn_hair_l_1}
      boxExtents=<22 8 8>
      constraintSetup {angularLimitsMax=<1 7 5>, angularLimitsMin=<-1 -20 -10>, angularTarget=<0 -18 0>}
      localJointOffset=<5 0 0>
     }]
    relativeSpaceBone {boneName=dyn_headAdjuster}
   }
   n55 = Anim Dynamics - Bone: dyn_hair_fr_1(ComponentPose=@n54, Alpha=0.6, ExternalForce="0,0,0") {
    angularDampingOverride=0.9999
    angularSpringConstant=6000
    bAngularSpring=true
    bOverrideAngularDamping=true
    bUseGravityOverride=true
    alphaBoolBlend {blendOption=Linear}
    boundBone {boneName=dyn_hair_fr_1}
    componentAppliedLinearAccClamp=<3000 3000 3000>
    componentLinearAccScale=<0.25 0.25 0.4>
    componentLinearVelScale=<0.25 0.25 0.4>
    gravityOverride {z=-400}
    physicsBodyDefinitions [{
      sphereCollisionRadius=0
      boundBone {boneName=dyn_hair_fr_1}
      boxExtents=<22 10 10>
      constraintSetup {angularLimitsMax=<1 5 20>, angularLimitsMin=<-1 -5 -5>, angularTarget=<0 0 15>}
      localJointOffset=<4 0 0>
     }]
    relativeSpaceBone {boneName=dyn_headAdjuster}
   }
   n56 = Anim Dynamics - Bone: dyn_hair_br_1(ComponentPose=@n55, Alpha=0.6, ExternalForce="0,0,0") {
    angularDampingOverride=0.9999
    angularSpringConstant=6000
    bAngularSpring=true
    bOverrideAngularDamping=true
    bUseGravityOverride=true
    alphaBoolBlend {blendOption=Linear}
    boundBone {boneName=dyn_hair_br_1}
    componentAppliedLinearAccClamp=<3000 3000 3000>
    componentLinearAccScale=<0.25 0.25 0.4>
    componentLinearVelScale=<0.25 0.25 0.4>
    gravityOverride {z=-400}
    physicsBodyDefinitions [{
      sphereCollisionRadius=0
      boundBone {boneName=dyn_hair_br_1}
      boxExtents=<15 5 5>
      constraintSetup {angularLimitsMax=<1 10 5>, angularLimitsMin=<-1 -10 -25>, angularTarget=<0 0 -20>}
      localJointOffset=<2 0 0>
     }]
    relativeSpaceBone {boneName=dyn_headAdjuster}
   }
   Output Pose(Result=Component To Local(ComponentPose=@n56)) {layerGroup=None}
   n57 = "Transform (Modify) Bone - Bone: dyn_headAdjuster"(
    ComponentPose=Local To Component(LocalPose=Copy Pose From Mesh() {bUseAttachedParent=true})
    Alpha=Get Neck Correct Alpha() -> (Neck Correct Alpha:double)
    Translation="-0.25,0.6,0"
    Rotation="0,1,0"
    Scale="1,1,1"
   ) {
    rotationMode=BMM_Additive
    rotationSpace=BCS_BoneSpace
    translationMode=BMM_Additive
    translationSpace=BCS_BoneSpace
    alphaBoolBlend {blendOption=Linear}
    boneToModify {boneName=dyn_headAdjuster}
    rotation {yaw=1}
    translation=<-0.25 0.6>
   }
   Save cached pose 'HeadRotation_Cache'(Pose=Component To Local(ComponentPose=@n57)) {cacheName=HeadRotation_Cache}
   Save cached pose 'Limit_Cache'() {cacheName=Limit_Cache}
   Use cached pose 'Limit_Cache'()
   Use cached pose 'Limit_Cache'()
  ]}
 EventGraph {nodes [
   Event BlueprintUpdateAnimation()
   > n1 = Sequence!(execute) -> (then_0, then_1, then_2:Exec)
   > n2 = Gravity Override State Function DEPRECATED(CharacterPartAnimInstance=Self-Reference(), GameTimeInSeconds=GetGameTimeInSeconds(), GravityOverrideParams=Get Gravity Override Params())
   > Set GravityOverride(GravityOverride=@n2)
   n3 = CheckForBackbling()
   n4 = Branch(execute=@n1.then_2, Condition=@n3)
   > Set Is Backbling Long(Is Backbling Long:bool=@n3.bIsBackblingLong) -> (Output_Get:bool)
   > Set isBackBling(isBackBling:bool=true) -> (Output_Get:bool)
   > Set Backbling Alpha(Backbling Alpha:double=SelectFloat(A=0, B=1, bPickA=@n3.bIsBackblingLong)) -> (Output_Get:double)
   Set isBackBling(execute=@n4.else, isBackBling=false)
   > Set Backbling Alpha(Backbling Alpha=0)
   n5 = BreakTransform(InTransform=GetDeltaTransformFromRefPose(self=GetOwnerSkeletalMeshComponent(), BoneName=head, BaseName=spine_05))
   Set Neck Correct Alpha(execute=@n1.then_1, Neck Correct Alpha:double=MapRangeClamped(Value=BreakRotator(InRot=@n5.Rotation).Yaw, InRangeA=-15, InRangeB=-45)) -> (Output_Get:double)
   n6 = BreakTransform(InTransform=GetDeltaTransformFromRefPose(self=GetOwnerSkeletalMeshComponent(), BoneName=neck_01, BaseName=spine_05))
   > Set Head_Left_Alpha(Head_Left_Alpha:double=MapRangeClamped(Value=BreakRotator(InRot=@n6.Rotation), InRangeA=0, InRangeB=25)) -> (Output_Get:double)
   > n7 = Set Head_Down_Alpha(Head_Down_Alpha:double=0) -> (Output_Get:double)
   n8 = Branch(execute=[@n13.else, @n7.then], Condition=Get bIsCrouching())
   n9 = GetOwnerSkeletalMeshComponent()
   n10 = BreakTransform(InTransform=GetDeltaTransformFromRefPose(self=@n9, BoneName=head, BaseName=spine_05))
   > Set Head_Crouch_Alpha(Head_Crouch_Alpha:double=MapRangeClamped(Value=BreakRotator(InRot=@n10.Rotation).Yaw, InRangeA=-20, InRangeB=-70, OutRangeA=0.2)) -> (Output_Get:double)
   > Set Head_Up_Alpha(Head_Up_Alpha:double=0) -> (Output_Get:double)
   > Set Head_Down_Alpha(Head_Down_Alpha=0.1)
   n11 = BreakRotator(InRot=BreakTransform(InTransform=GetDeltaTransformFromRefPose(self=@n9, BoneName=spine_05, BaseName=pelvis)).Rotation)
   > Set Spine_Rot(Spine_Rot:double=MapRangeClamped(Value=@n11.Yaw, InRangeA=30, InRangeB=50, OutRangeB=0.5)) -> (Output_Get:double)
   Set Head_Crouch_Alpha(execute=@n8.else, Head_Crouch_Alpha=0.4)
   n12 = BreakTransform(InTransform=GetDeltaTransformFromRefPose(self=GetOwnerSkeletalMeshComponent(), BoneName=head, BaseName=spine_05))
   > Set Head_Up_Alpha(Head_Up_Alpha=MapRangeClamped(Value=BreakRotator(InRot=@n12.Rotation).Yaw, InRangeA=-15, InRangeB=-45))
   > Set Spine_Rot(Spine_Rot=MapRangeClamped(Value=@n11.Yaw, InRangeA=0, InRangeB=30))
   n13 = Branch(Condition=Get Is Backbling Long())
   > Set Spine_Rot(Spine_Rot=0)
   > Set Head_Crouch_Alpha(Head_Crouch_Alpha=0)
   > Set Head_Up_Alpha(Head_Up_Alpha=0)
   > Set Head_Down_Alpha(Head_Down_Alpha=0)
   n14 = GetDeltaTransformFromRefPose(self=GetOwnerSkeletalMeshComponent(), BoneName=clavicle_r, BaseName=spine_05)
   n15 = MapRangeClamped(Value=BreakRotator(InRot=BreakTransform(InTransform=@n14).Rotation).Pitch, InRangeA=15, InRangeB=45)
   n16 = GetDeltaTransformFromRefPose(self=GetOwnerSkeletalMeshComponent(), BoneName=clavicle_l, BaseName=spine_05)
   n17 = MapRangeClamped(Value=BreakRotator(InRot=BreakTransform(InTransform=@n16).Rotation).Pitch, InRangeA=15, InRangeB=45)
   Set Shldr Up Adjust(Shldr Up Adjust:double=SelectFloat(A=@n17, B=@n15, bPickA="float >= float"(A=@n17, B=@n15))) -> (Output_Get:double)
  ]}
]
variables [
 Backbling Alpha {type=number}
 Gravity Override Params {
  type=GravityOverrideParamsStruct
  default {
   emoteOrMelee {gravityOverride=<0 0 -800>}
   onGround_Moving {gravityOverride=<0 0 -800>}
   onGround_Standing {gravityOverride=<0 0 -800>}
   skydive_DivingDown {windAmplitude=<50 80 60>, windFrequency=<0.5 3 2>}
   skydive_FloatingDown {gravityOverride=<0 0 -200>, windAmplitude=<50 -75 40>, windFrequency=<1 2 2.5>, windOffset=<0 0 0>}
   skydive_GoingUp {gravityOverride=<0 0 -200>, windAmplitude=<100 -250 200>, windFrequency=<0.75 3.5 4>}
   skydive_Parachuting {gravityOverride=<0 0 -720>, windAmplitude=<30 -50 45>, windFrequency=<0.25 0.75 0.8>}
  }
 }
 GravityOverride {type=Vector}
 Head_Crouch_Alpha {type=number}
 Head_Down_Alpha {type=number}
 Head_Left_Alpha {type=number}
 Head_Right_Alpha {type=number}
 Head_Up_Alpha {type=number}
 Is Backbling Long {type=boolean}
 isBackBling {type=boolean}
 Neck Correct Alpha {type=number}
 Shldr Up Adjust {type=number}
 Spine_Rot {type=number}
]
